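# Translucent two-sided material for an editor perf-tool volume that displays an actor count on a tinted box with highlighted edges.
# ActorCount and MaxDigits arrive as custom primitive data and feed DebugScalarValues, using scale-adjusted centred UVs, masked to upward faces.
# Mirrored-UV edge thresholds combine with the digits into one mask driving colour and opacity.
# A Fresnel-like view-angle blend tints the interior, and depth fade softens intersections.
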Material /Game/NeverCook/EditorUtilities/PerfTools/ActorCountVolume/M_ActorCountVolume
graph {
 properties {blendMode=BLEND_Translucent, pixelDepthOffsetMode=PDOM_Legacy, twoSided=true, usedShadingModels=Default Lit}
 nodes [
  n1 = If(
   A=ComponentMask(None=NamedRerouteUsage()) {r=true}
   B=ScalarParameter() {defaultValue=0.98, parameterName=UVEdgeClamp}
   "A > B"=Constant() {r=1}
   "A < B"=Constant()
  ) {constB=0.5}
  n2 = If(
   A=ComponentMask(None=NamedRerouteUsage()) {g=true}
   B=ScalarParameter() {defaultValue=0.98, parameterName=UVEdgeClamp}
   "A > B"=Constant() {r=1}
   "A < B"=Constant()
  ) {constB=0.5}
  n3 = Saturate(None=Add(A=Add(A=@n1, B=@n2), B=NamedRerouteUsage()))
  DepthFade(Opacity=LinearInterpolate(A=ScalarParameter() {defaultValue=0.5, parameterName=Opacity_Facing}, Alpha=@n3))
  n4 = VectorParameter() {parameterName=EdgeColor, defaultValue=<rgba 0.9375 0.9375 0.9375 1>}
  LinearInterpolate(A=@n4.RGB, B=LinearInterpolate(A=NamedRerouteUsage(), B=@n4.RGB, Alpha=@n3), Alpha=DepthFade() {fadeDistanceDefault=500})
  n5 = Subtract(A=TextureCoordinate()) {constB=0.5}
  NamedRerouteDeclaration(None=Abs(None=Multiply(A=@n5) {constB=2})) {name=UV, nodeColor=<rgba 1 0 0.0937 1>}
  n6 = Subtract(A=CameraPositionWS(), B=WorldPosition().XYZ)
  n7 = VertexNormalWS()
  n8 = Power(Base=DotProduct(A=Normalize(VectorInput=@n6), B=Normalize(VectorInput=@n7))) {constExponent=0.1}
  NamedRerouteDeclaration(None=LinearInterpolate(
    A=VectorParameter() {parameterName=Color_Edge, defaultValue=<rgba 0.2959 0.3638 0.5521 1>}.RGB
    B=VectorParameter() {parameterName=Color_Facing, defaultValue=<rgba 0.2604 0.2604 0.2604 1>}.RGB
    Alpha=Saturate(None=@n8)
   )) {name=FinalInteriorColor, nodeColor=<rgba 0.9688 0 1 1>}
  n9 = ScalarParameter() {bUseCustomPrimitiveData=true, parameterName=ActorCount}
  n10 = ScalarParameter() {bUseCustomPrimitiveData=true, defaultValue=5, parameterName=MaxDigits, primitiveDataIndex=1}
  n11 = ScalarParameter() {defaultValue=0.4, parameterName=NumScale}
  n12 = MaterialFunctionCall() {materialFunction=&ObjectScale}
  n13 = Divide(A=AppendVector(A=@n12.Scale X, B=@n12.Scale X)) {constB=100}
  n14 = MaterialFunctionCall(Float2=@n13) {materialFunction=&BreakOutFloat2Components}
  n15 = MaterialFunctionCall(Texture Scale=Multiply(A=Divide(A=@n11, B=@n13), B=Max(A=@n14.R, B=@n14.R))) {materialFunction=&ScaleUVsByCenter}
  n16 = VertexNormalWS()
  n17 = Constant3Vector() {constant=<rgba 0 0 1 0>}
  NamedRerouteDeclaration(None=LinearInterpolate(
    B=MaterialFunctionCall(Number=@n9, MaximumNumberOfDigits=@n10, UVs=@n15.UVs) {materialFunction=&DebugScalarValues}.Result
    Alpha=DotProduct(A=@n16, B=@n17)
   )) {name=debugNumbers, nodeColor=<rgba 1 0 0.8203 1>}
 ]
}
