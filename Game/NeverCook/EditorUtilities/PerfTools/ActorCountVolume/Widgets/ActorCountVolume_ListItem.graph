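# List row widget for the Actor Count Volume editor perf tool, showing a label, count and percentage.
# ValidateData reads depth, count and divisor from the generic list item's integer array, indenting child rows with a circle icon, and sets the tooltip from its first string.
# Percentage text lerps green to red as the share rises.
# It extends TA_BaseEditorButtonWidget, tinting the bottom edge by selection and child state and filtering by label substring.
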WidgetBlueprint /Game/NeverCook/EditorUtilities/PerfTools/ActorCountVolume/Widgets/ActorCountVolume_ListItem
graphs [
 EventGraph {nodes [
   n1 = Event updateAfterDataChange!() -> (OutputDelegate, then, currentListItemData:Object Reference) {is=title K2Node_Event}
   > Parent: updateAfterDataChange!(execute, currentListItemData:Object Reference=@n1.currentListItemData) -> (then) {is=title K2Node_CallParentFunction}
   > n2 = Cast To TA_GenericListItemData!(execute, Object=@n1.currentListItemData) -> (then, CastFailed, AsTA Generic List Item Data:TA Generic List Item Data Object Reference) {is=title K2Node_DynamicCast}
   > Is Valid(InputObject=@n2)
   > ValidateData!(execute, self, data:TA Generic List Item Data Object Reference=@n2) -> (then)
  ]}
 FilterBySearchString {nodes [
   n1 = FilterBySearchString!() -> (then, SearchString:String) {is=title K2Node_FunctionEntry}
   > Return Node!(execute, ShouldDisplay:bool=Contains(SearchIn=Get label() -> (label:String), Substring=@n1))
  ]}
 ValidateData {nodes [
   n1 = ValidateData!() -> (then, data:TA Generic List Item Data Object Reference) {is=title K2Node_FunctionEntry}
   > Set l_data(l_data:TA Generic List Item Data Object Reference=@n1) -> (Output_Get:TA Generic List Item Data Object Reference)
   > n2 = Sequence!(execute) -> (then_0, then_1, then_2:Exec)
   n3 = Get integers!(self:TA Generic List Item Data Object Reference=Get l_data() -> (l_data:TA Generic List Item Data Object Reference)) -> (integers:Array of Integers)
   > Set l_depth(l_depth:int=int + int(A="Get (a copy)"(Array:Array of Integers=@n3) -> (Output:int))) -> (Output_Get:int=0)
   > Set IsChild(IsChild:bool="integer > integer"(A=Get l_depth() -> (l_depth:int=0), B=1)) -> (Output_Get:bool)
   > SetVisibility(
    self=Get imgIcon() -> (imgIcon:Image Object Reference)
    InVisibility=Select(Option 0=Hidden, Option 1=Visible, Index=Get IsChild() -> (IsChild:bool))
   )
   > Branch(Condition=Get IsChild())
   n4 = float * float(A:int=Get l_depth() -> (l_depth=0), B=10) {is=title K2Node_PromotableOperator}
   > SetPadding!(
    execute
    self:Horizontal Box Slot Object Reference=SlotAsHorizontalBoxSlot!(Widget:Widget Object Reference=Get imgIcon()) -> (ReturnValue:Horizontal Box Slot Object Reference)
    InPadding:Margin Structure=Make Margin(Left=@n4, Top=0, Right=0, Bottom=0)
   ) -> (then)
   n5 = Get Brush!(self:Image Object Reference=Get imgIcon()) -> (Brush:Slate Brush Structure)
   n6 = float * float(
    A:double=BreakVector2D(InVec=Break Slate Brush!(SlateBrush:"Slate Brush Structure (by ref)"=@n5) -> (
      TintColor:Slate Color Structure
      DrawAs:ESlateBrushDrawType Enum
      Tiling:ESlateBrushTileType Enum
      Mirroring:ESlateBrushMirrorType Enum
      ImageSize:Vector 2D Structure
      Margin:Margin Structure
      ResourceObject:Object Reference
      OutlineSettings:Slate Brush Outline Settings Structure
     ) {is=title K2Node_BreakStruct}.ImageSize)
    B=2
   )
   > SetPadding(
    self:Vertical Box Slot Object Reference=SlotAsVerticalBoxSlot(Widget=Get imageEdge() -> (imageEdge:Image Object Reference))
    InPadding=Make Margin(Left=float + float(A=@n4, B=@n6), Top=0, Right=0, Bottom=0)
   )
   n7 = Get l_data()
   Set l_count(execute=@n2.then_1, l_count:int="Get (a copy)"(Array=Get integers(self=@n7), Dimension 1=1)) -> (Output_Get:int=0)
   > "SetText (Text)"(self=Get txtCount() -> (txtCount:Text Object Reference), InText="To Text (Integer)"(Value=Get l_count() -> (l_count:int=0)))
   n8 = Get integers(self=Get l_data())
   n9 = float / float(A="To Float (Integer)"(InInt=Get l_count() -> (l_count=0)), B="To Float (Integer)"(InInt="Get (a copy)"(Array=@n8, Dimension 1=2)))
   > "SetText (Text)"(
    self=Get txtPer() -> (txtPer:Text Object Reference)
    InText=As Percent!(
     Value:float=@n9
     RoundingMode:ERoundingMode Enum
     bAlwaysSign:bool=false
     bUseGrouping:bool=true
     MinimumIntegralDigits:int=1
     MaximumIntegralDigits:int=324
     MinimumFractionalDigits:int=0
     MaximumFractionalDigits:int=1
    ) -> (ReturnValue:Text)
   )
   > SetColorAndOpacity(
    self=Get txtPer()
    InColorAndOpacity=Make Slate Color(SpecifiedColor="Lerp (LinearColor)"(A=(R=0.1168,G=1,B=0.0821,A=1), B=(R=1,G=0,B=0,A=1), Alpha=@n9))
   )
   n10 = Get strings!(self:TA Generic List Item Data Object Reference=Get l_data()) -> (strings:Array of Strings)
   Branch(execute=@n2.then_2, Condition=Is Valid Index!(TargetArray=@n10, IndexToTest:int) -> (ReturnValue:bool))
   > Set ToolTipText(ToolTipText:Text="To Text (String)"(InString="Get (a ref)"(Array:Array of Strings=@n10, Dimension 1=0) -> (Output:"String (by ref)"))) -> (Output_Get:Text)
  ]}
 getBackgroundWidget {nodes [
   getBackgroundWidget() {is=title K2Node_FunctionEntry}
   > Return Node!(execute, backgroundWidget:Widget Object Reference=Get borderMain() -> (borderMain:Border Object Reference))
  ]}
 getLabelWidget {nodes [
   getLabelWidget() {is=title K2Node_FunctionEntry}
   > Return Node!(execute, labelWidget:Widget Object Reference=Get txtMain() -> (txtMain:Text Object Reference))
  ]}
 setHorizontalAlignmentSet {nodes [Parent: setHorizontalAlignmentSet!(
    execute
    alignment:EHorizontalAlignment Enum=setHorizontalAlignmentSet!() -> (then, alignment:EHorizontalAlignment Enum=HAlign_Fill) {is=title K2Node_FunctionEntry}
   ) -> (then) {is=title K2Node_CallParentFunction}]}
 validateBackgroundColor {nodes [
   validateBackgroundColor() {is=title K2Node_FunctionEntry}
   n1 = Get IsChild()
   > SetColorAndOpacity(
    self:Image Object Reference=Get imageEdge()
    InColorAndOpacity:Linear Color Structure=SelectColor(
     A=(R=0.0284,G=0.3763,B=1,A=1)
     B=SelectColor(A=(R=1,G=0.9917,B=0.9104,A=0.5), B=(R=0.8123,G=0.8217,B=0.8385,A=1), bPickA=@n1)
     bPickA=getSelected!() -> (isSelected:bool)
    )
   )
   > n2 = Parent: validateBackgroundColor!(execute) -> (then, handled:bool) {is=title K2Node_CallParentFunction}
   > Return Node!(execute, handled:bool=@n2)
  ]}
]
variables [IsChild {type=boolean}]
widgets {
 info {
  inheritedWidgetCount=0
  parentClass=/Game/Blueprints/Blutilities/ForTechArt/Widgets/Buttons/TA_BaseEditorButtonWidget.TA_BaseEditorButtonWidget_C
  rootWidgetClass=&Overlay
  widgetCount=12
 }
 widgets [
  Overlay_7 {widgetClassPath=&Overlay, widget=Overlay_7}
  VerticalBox_0 {
   widgetClassPath=&VerticalBox
   parent=Overlay_7
   widget=VerticalBox_0
   slot {class=OverlaySlot, properties {horizontalAlignment=HAlign_Fill}}
  }
  borderMain {
   widgetClassPath=&Border
   bIsVariable=true
   parent=VerticalBox_0
   widget=borderMain
   properties {padding=<ltrb 0 0 3 0>}
   slot {class=VerticalBoxSlot}
  }
  HorizontalBox_0 {
   widgetClassPath=&HorizontalBox
   parent=borderMain
   widget=HorizontalBox_0
   slot {class=BorderSlot, properties {padding=<ltrb 0 0 3 0>}}
  }
  imgIcon {
   widgetClassPath=&Image
   bIsVariable=true
   parent=HorizontalBox_0
   widget=imgIcon
   properties {
    visibility=Collapsed
    brush {
     bIsDynamicallyLoaded=false
     drawAs=Image
     imageType=NoImage
     mirroring=NoMirror
     resourceName=None
     resourceObject=/Game/UI/Fort_Traits/Circle
     tiling=NoTile
     imageSize=<12 12>
     margin=<ltrb 0 0 0 0>
     outlineSettings {
      bUseBrushTransparency=false
      roundingType=HalfHeightRadius
      width=0
      color {colorUseRule=UseColor_Specified, specifiedColor=<rgba 0 0 0 0>}
      cornerRadii=<0 0 0 0>
     }
     tintColor {colorUseRule=UseColor_Specified, specifiedColor=<rgba 1 1 1 1>}
     uVRegion {bIsValid=false, max=<0 0>, min=<0 0>}
    }
   }
   slot {class=HorizontalBoxSlot, properties {horizontalAlignment=HAlign_Center, verticalAlignment=VAlign_Center}}
  }
  ScaleBox_0 {
   widgetClassPath=&ScaleBox
   parent=HorizontalBox_0
   widget=ScaleBox_0
   properties {clipping=ClipToBounds, stretchDirection=DownOnly}
   slot {class=HorizontalBoxSlot, properties {padding=<ltrb 0 5 0 5>, size {sizeRule=Fill, value=1}}}
  }
  txtMain {
   widgetClassPath=&TextBlock
   bIsVariable=true
   parent=ScaleBox_0
   widget=txtMain
   properties {
    clipping=ClipToBounds
    justification=Right
    text=Test
    colorAndOpacity {colorUseRule=UseColor_Specified, specifiedColor=<rgba 0.0104 0.0104 0.0104 1>}
    font {
     bForceMonospaced=false
     bMaterialIsStencil=false
     fontMaterial=null
     fontObject=&Roboto
     letterSpacing=0
     monospacedWidth=1
     size=10
     skewAmount=0
     typefaceFontName=Bold
     outlineSettings {
      bApplyOutlineToDropShadows=false
      bMiteredCorners=false
      bSeparateFillAlpha=false
      outlineMaterial=null
      outlineSize=0
      outlineColor=<rgba 0 0 0 1>
     }
    }
    shadowColorAndOpacity=<rgba 0 0 0 0.2>
   }
   slot {class=ScaleBoxSlot, properties {horizontalAlignment=HAlign_Fill}}
  }
  imageEdge_2 {
   widgetClassPath=&Image
   bIsVariable=true
   parent=HorizontalBox_0
   widget=imageEdge_2
   properties {
    brush {
     bIsDynamicallyLoaded=false
     drawAs=Image
     imageType=NoImage
     mirroring=NoMirror
     resourceName=None
     resourceObject=null
     tiling=NoTile
     imageSize=<1 0>
     margin=<ltrb 0 0 0 0>
     outlineSettings {
      bUseBrushTransparency=false
      roundingType=HalfHeightRadius
      width=0
      color {colorUseRule=UseColor_Specified, specifiedColor=<rgba 0 0 0 0>}
      cornerRadii=<0 0 0 0>
     }
     tintColor {colorUseRule=UseColor_Specified, specifiedColor=<rgba 1 1 1 1>}
     uVRegion {bIsValid=false, max=<0 0>, min=<0 0>}
    }
    colorAndOpacity=<rgba 0.3542 0.3542 0.3542 1>
   }
   slot {class=HorizontalBoxSlot, properties {horizontalAlignment=HAlign_Right, padding=<ltrb 3 0 3 0>}}
  }
  txtCount {
   widgetClassPath=&TextBlock
   bIsVariable=true
   parent=HorizontalBox_0
   widget=txtCount
   properties {
    clipping=ClipToBounds
    justification=Center
    minDesiredWidth=45
    text=55555
    colorAndOpacity {colorUseRule=UseColor_Specified, specifiedColor=<rgba 0.8646 0.8646 0.8646 1>}
    font {
     bForceMonospaced=false
     bMaterialIsStencil=false
     fontMaterial=null
     fontObject=&Roboto
     letterSpacing=0
     monospacedWidth=1
     size=9.75
     skewAmount=0
     typefaceFontName=Bold
     outlineSettings {
      bApplyOutlineToDropShadows=false
      bMiteredCorners=false
      bSeparateFillAlpha=false
      outlineMaterial=null
      outlineSize=0
      outlineColor=<rgba 0 0 0 1>
     }
    }
    shadowColorAndOpacity=<rgba 0 0 0 0.2>
   }
   slot {class=HorizontalBoxSlot, properties {verticalAlignment=VAlign_Center}}
  }
  imageEdge_1 {
   widgetClassPath=&Image
   bIsVariable=true
   parent=HorizontalBox_0
   widget=imageEdge_1
   properties {
    brush {
     bIsDynamicallyLoaded=false
     drawAs=Image
     imageType=NoImage
     mirroring=NoMirror
     resourceName=None
     resourceObject=null
     tiling=NoTile
     imageSize=<1 0>
     margin=<ltrb 0 0 0 0>
     outlineSettings {
      bUseBrushTransparency=false
      roundingType=HalfHeightRadius
      width=0
      color {colorUseRule=UseColor_Specified, specifiedColor=<rgba 0 0 0 0>}
      cornerRadii=<0 0 0 0>
     }
     tintColor {colorUseRule=UseColor_Specified, specifiedColor=<rgba 1 1 1 1>}
     uVRegion {bIsValid=false, max=<0 0>, min=<0 0>}
    }
    colorAndOpacity=<rgba 0.3542 0.3542 0.3542 1>
   }
   slot {class=HorizontalBoxSlot, properties {padding=<ltrb 3 0 3 0>}}
  }
  txtPer {
   widgetClassPath=&TextBlock
   bIsVariable=true
   parent=HorizontalBox_0
   widget=txtPer
   properties {
    clipping=ClipToBounds
    justification=Right
    minDesiredWidth=36
    text=100%
    colorAndOpacity {colorUseRule=UseColor_Specified, specifiedColor=<rgba 0.8646 0.8646 0.8646 1>}
    font {
     bForceMonospaced=false
     bMaterialIsStencil=false
     fontMaterial=null
     fontObject=&Roboto
     letterSpacing=0
     monospacedWidth=1
     size=9.75
     skewAmount=0
     typefaceFontName=Bold
     outlineSettings {
      bApplyOutlineToDropShadows=false
      bMiteredCorners=false
      bSeparateFillAlpha=false
      outlineMaterial=null
      outlineSize=0
      outlineColor=<rgba 0 0 0 1>
     }
    }
    shadowColorAndOpacity=<rgba 0 0 0 0.2>
   }
   slot {class=HorizontalBoxSlot, properties {verticalAlignment=VAlign_Center, padding=<ltrb 0 0 3 0>}}
  }
  imageEdge {
   widgetClassPath=&Image
   bIsVariable=true
   parent=VerticalBox_0
   widget=imageEdge
   properties {
    brush {
     bIsDynamicallyLoaded=false
     drawAs=Image
     imageType=NoImage
     mirroring=NoMirror
     resourceName=None
     resourceObject=null
     tiling=NoTile
     imageSize=<32 2>
     margin=<ltrb 0 0 0 0>
     outlineSettings {
      bUseBrushTransparency=false
      roundingType=HalfHeightRadius
      width=0
      color {colorUseRule=UseColor_Specified, specifiedColor=<rgba 0 0 0 0>}
      cornerRadii=<0 0 0 0>
     }
     tintColor {colorUseRule=UseColor_Specified, specifiedColor=<rgba 1 1 1 1>}
     uVRegion {bIsValid=false, max=<0 0>, min=<0 0>}
    }
    colorAndOpacity=<rgba 0.2031 0.2031 0.2031 1>
   }
   slot {class=VerticalBoxSlot}
  }
 ]
}
