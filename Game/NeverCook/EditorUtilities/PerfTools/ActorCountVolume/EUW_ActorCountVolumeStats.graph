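# Editor utility widget that shows a per-class breakdown of actor counts for the selected EUA_ActorCountVolume.
# It follows level editor selection and the volume's count-update event to rebuild its data from the volume's actor descriptors.
# Descriptors are tallied by parent class, using the volume's transient class filter or a default, mostly Building, class list.
# The list shows total, per-class and Other rows, and a footer sums selected rows against the total.
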EditorUtilityWidgetBlueprint /Game/NeverCook/EditorUtilities/PerfTools/ActorCountVolume/EUW_ActorCountVolumeStats
graphs [
 EventGraph {nodes [
   n1 = "On Item Selection Changed (TA_Generic01_List)"!() -> (
    OutputDelegate
    then
    selectedData:Array of Object References
    selectedWidgets:Array of User Widget Object References
    selectedIndices:Array of Integers
   ) {is=title K2Node_ComponentBoundEvent}
   > UpdateInfoText!(execute, self, Array:Array of Object References=@n1.selectedData) -> (then)
   n2 = HandleSelectionChanged!() -> (OutputDelegate, then, SelectionSet:Typed Element Selection Set Object Reference) {is=title K2Node_CustomEvent}
   > ValidateSelection()
   Event Construct()
   > n3 = GetSelectionSet(self=Get LevelEditorSubsystem())
   > Bind Event to On Selection Change(self:Typed Element Selection Set Object Reference=@n3, Delegate=@n2) {is=title K2Node_AddDelegate}
   > ValidateSelection()
  ]}
 GetActorDescriptions {nodes [
   n1 = GetActorDescriptions!() -> (then, useClassFilter:bool=~) {is=title K2Node_FunctionEntry}
   > Is Valid(InputObject=Get CurrentActor() -> (CurrentActor:EUA Actor Count Volume Object Reference))
   > n2 = GetActorDescriptions!(execute, self:EUA Actor Count Volume Object Reference=Get CurrentActor(), useClassFilter:bool=@n1) -> (then, count:int=0, descs:Array of Actor Desc Structures) {is=title K2Node_CallFunction}
   > Return Node!(execute, CurrentActorDescs:Array of Actor Desc Structures=@n2.descs)
  ]}
 HandleCountChanged {nodes [
   n1 = HandleCountChanged!() -> (then, actor:EUA Actor Count Volume Object Reference) {is=title K2Node_FunctionEntry}
   > UpdateInfoTarget!(execute, self, actor:EUA Actor Count Volume Object Reference=@n1) -> (then)
  ]}
 IsClassIncluded {nodes [
   n1 = IsClassIncluded!() -> (then, desc:Actor Desc Structure) {is=title K2Node_FunctionEntry}
   > n2 = For Each Loop(Array:Array of Actor Class References=Get ClassList() -> (ClassList:Array of Actor Class References)) -> (Array Element:Actor Class Reference)
   Return Node!(execute=@n2.Completed, include:bool=false, class:Actor Class Reference)
   n3 = ClassIsChildOf(
    TestClass=Break Actor Desc!(ActorDesc:"Actor Desc Structure (by ref)"=@n1) -> (
     Guid:Guid Structure
     NativeClass:Object Class Reference
     Class:Soft Object Path Structure
     Name:Name
     Label:Name
     Bounds:Box Structure
     RuntimeGrid:Name
     bIsSpatiallyLoaded:bool=~
     bActorIsEditorOnly:bool=~
     ActorPackage:Name
     ActorPath:Name
     FolderPath:Name
     DataLayerAssets:Array of Soft Object Path Structures
    ) {is=title K2Node_BreakStruct}.NativeClass
    ParentClass=@n2
   )
   > Branch(Condition=@n3)
   > Return Node(include=@n3, class=@n2)
  ]}
 Reset {nodes [
   Reset() {is=title K2Node_FunctionEntry}
   > Clear!(
    execute
    TargetMap:Map of Object Class References to Integers=Get ClassToCountMap() -> (ClassToCountMap:Map of Object Class References to Integers)
   ) -> (then) {is=title K2Node_CallFunction}
   > clearItems(self:TA Generic 01 List Object Reference=Get TA_Generic01_List() -> (TA_Generic01_List:TA Generic 01 List Object Reference))
   > Set OtherCount(OtherCount:int=0) -> (Output_Get:int=0)
   > "SetText (Text)"(self=Get txtInfo() -> (txtInfo:Text Object Reference))
  ]}
 TryAddingActorData {nodes [
   n1 = TryAddingActorData!() -> (then, ActorDesc:"Actor Desc Structure (by ref)") {is=title K2Node_FunctionEntry}
   > Set l_desc(l_desc:Actor Desc Structure=@n1) -> (Output_Get:Actor Desc Structure)
   n2 = Get CurrentActor()
   > Is Valid(InputObject=@n2)
   n3 = Get TransientClassFilter!(self:EUA Actor Count Volume Object Reference=@n2) -> (TransientClassFilter:Array of Object Class References)
   > n4 = Branch(Condition=Is Valid Index!(TargetArray:Array of Object Class References=@n3, IndexToTest:int) -> (ReturnValue:bool))
   > n5 = IsDescInClassList!(
    execute
    ActorDesc:Actor Desc Structure=Get l_desc() -> (l_desc:Actor Desc Structure)
    ParentClasses:Array of Object Class References=@n3
    __WorldContext:Object Reference
   ) -> (then, IsInList:bool, Class:Object Class Reference, ActorDescOut:Actor Desc Structure)
   > Set l_valid(l_valid:bool=@n5) -> (Output_Get:bool)
   > n6 = Set l_class(l_class:Object Class Reference=@n5.Class) -> (Output_Get:Object Class Reference)
   n7 = Branch(execute=[@n15.then, @n6.then], Condition=Get l_valid() -> (l_valid:bool))
   n8 = Find(
    TargetMap:Map of Object Class References to Integers=Get ClassToCountMap()
    Key:"Object Class Reference (by ref)"=Get l_class() -> (l_class:Object Class Reference)
   ) -> (Value:int)
   > n9 = Branch(Condition=@n8.ReturnValue)
   > n10 = Set l_count(l_count:int=int + int(A=@n8)) -> (Output_Get:int=0)
   n11 = Get l_class()
   n12 = Get l_count() -> (l_count:int=0)
   Add!(
    execute=[@n10.then, @n13.then]
    TargetMap:Map of Object Class References to Integers=Get ClassToCountMap()
    Key:"Object Class Reference (by ref)"=@n11
    Value:"Integer (by ref)"=@n12
   ) -> (then) {is=title K2Node_CallFunction}
   > PrintString(InString=Append!(A="Adding ", B=Get Class Display Name(Class=@n11), C:String=" with count ", D:String="To String (Integer)"(InInt=@n12)) -> (ReturnValue))
   n13 = Set l_count(execute=@n9.else, l_count=1) -> (Output_Get=0)
   Increment Int(" "=@n7.else, Value=Get OtherCount() -> (OtherCount:int=0))
   n14 = IsClassIncluded!(execute=@n4.else, self, desc:Actor Desc Structure=Get l_desc()) -> (then, include:bool, class:Actor Class Reference) {is=title K2Node_CallFunction}
   > Set l_valid(l_valid=@n14)
   > n15 = Set l_class(l_class=@n14.class)
   n16 = Find(TargetMap=Get OtherClasses() -> (OtherClasses:Map of Object Class References to Integers), Key=Get l_class())
   n17 = Branch(Condition=@n16.ReturnValue)
   > n18 = Set l_count(l_count=int + int(A=@n16)) -> (Output_Get=0)
   Add(execute=[@n19.then, @n18.then], TargetMap=Get OtherClasses(), Key=Get l_class(), Value=Get l_count() -> (l_count=0))
   n19 = Set l_count(execute=@n17.else, l_count=1) -> (Output_Get=0)
   Get l_class()
  ]}
 UpdateInfoTarget {nodes [
   n1 = UpdateInfoTarget!() -> (then, actor:EUA Actor Count Volume Object Reference) {is=title K2Node_FunctionEntry}
   > Reset() {is=title K2Node_CallFunction}
   > n2 = Sequence!(execute) -> (then_0, then_1, then_2:Exec)
   > n3 = Set CurrentActor(CurrentActor:EUA Actor Count Volume Object Reference=@n1) -> (Output_Get:EUA Actor Count Volume Object Reference)
   > n4 = Is Valid(InputObject=@n3)
   Return Node!(execute=@n4.Is Not Valid)
   n5 = GetActorDescriptions!(execute=@n2.then_1, self:Self Object Reference, useClassFilter=false) -> (then, CurrentActorDescs:Array of Actor Desc Structures)
   > n6 = Set l_descs(l_descs:Array of Actor Desc Structures=@n5) -> (Output_Get:Array of Actor Desc Structures)
   > Branch(Condition=Is Valid Index(TargetArray:Array of Actor Desc Structures=@n6))
   > n7 = Sequence()
   n8 = Get TransientClassFilter(self=Get CurrentActor())
   > Set DisplayClassList(DisplayClassList:Array of Object Class References=Select(
     Option 0:Array of Object Class References=Get ClassList()
     Option 1:Array of Object Class References=@n8
     Index=Is Valid Index(TargetArray:Array of Object Class References=@n8)
    ) -> (ReturnValue:Array of Object Class References)) -> (Output_Get:Array of Object Class References)
   n9 = For Each Loop(Exec=@n7.then_1, Array:Array of Actor Desc Structures=Get l_descs() -> (l_descs:Array of Actor Desc Structures)) -> (Array Element:Actor Desc Structure)
   > TryAddingActorData!(execute, self, ActorDesc:"Actor Desc Structure (by ref)"=@n9) -> (then) {is=title K2Node_CallFunction}
   UpdateList(execute=@n7.then_2)
   n10 = GetActorLabel(self=Get CurrentActor(), bCreateIfNone=true)
   "SetText (Text)"(
    execute=@n2.then_2
    self=Get txtActor() -> (txtActor:Text Object Reference)
    InText="To Text (String)"(InString=Append!(A="Selected Volume: ", B=@n10) -> (ReturnValue))
   )
  ]}
 UpdateInfoText {nodes [
   n1 = UpdateInfoText!() -> (then, Array:Array of Object References) {is=title K2Node_FunctionEntry}
   > Is Valid(InputObject=Get CurrentActor())
   > n2 = For Each Loop(Array:Array of Object References=@n1) -> (Array Element:Object Reference)
   n3 = "To String (Integer)"(InInt=Get l_countSelected() -> (l_countSelected:int=0))
   n4 = "To String (Integer)"(InInt=Get CurrentActorCount!(self:EUA Actor Count Volume Object Reference=Get CurrentActor()) -> (CurrentActorCount:int=0))
   "SetText (Text)"(
    execute=@n2.Completed
    self=Get txtInfo()
    InText="To Text (String)"(InString=Append!(A="Total Selected In List: ", B=@n3, C:String=" / ", D:String=@n4) -> (ReturnValue))
   )
   n5 = Cast To TA_GenericListItemData!(execute=@n2.LoopBody, Object=@n2) -> (then, CastFailed, AsTA Generic List Item Data:TA Generic List Item Data Object Reference) {is=title K2Node_DynamicCast}
   n6 = Get integers!(self:TA Generic List Item Data Object Reference=@n5) -> (integers:Array of Integers)
   > Set l_countSelected(l_countSelected:int=int + int(A="Get (a ref)"(Array=@n6, Dimension 1=1), B=Get l_countSelected() -> (l_countSelected=0))) -> (Output_Get:int=0)
  ]}
 UpdateList {nodes [
   UpdateList() {is=title K2Node_FunctionEntry}
   > Is Valid(InputObject=Get CurrentActor())
   > n1 = Sequence!(execute) -> (then_0, then_1, then_2, then_3:Exec)
   > n2 = Construct Object from Class(Class=/Game/Blueprints/Blutilities/ForTechArt/Data/TA_GenericListItemData.TA_GenericListItemData_C) -> (ReturnValue:TA Generic List Item Data Object Reference)
   n3 = Get CurrentActorCount(self=Get CurrentActor()) -> (CurrentActorCount=0)
   > Set integers!(
    execute
    integers:Array of Integers=Make Array!("[0]":int=0, "[1]":int=@n3, "[2]":int=@n3) -> (Array:Array of Integers)
    self:TA Generic List Item Data Object Reference=@n2
   ) -> (then, Output_Get:Array of Integers)
   > Set displayString!(execute, displayString:String=Total Actor Count, self:TA Generic List Item Data Object Reference=@n2) -> (then, Output_Get:String)
   > n4 = addItem!(execute, self:TA Generic 01 List Object Reference=Get TA_Generic01_List(), dataObject:Object Reference=@n2) -> (then, newItem:User Widget Object Reference, newData:Object Reference)
   > SetVisibility(self=@n4, InVisibility=HitTestInvisible)
   PrintString(execute=@n1.then_1, InString=Classes)
   > n5 = Keys!(execute, TargetMap:Map of Object Class References to Integers=Get ClassToCountMap()) -> (then, Keys:Array of Object Class References)
   > n6 = For Each Loop(Array:Array of Object Class References=@n5) -> (Array Element:Object Class Reference)
   n7 = For Each Loop(Exec=@n6.Completed, Array=Get DisplayClassList() -> (DisplayClassList:Array of Object Class References))
   n8 = Find(TargetMap=Get ClassToCountMap(), Key=@n7)
   > PrintString(InString=Append(A="lOOKING FOR ", B=Get Class Display Name(Class=@n7), C=" and found: ", D="To String (Integer)"(InInt=@n8)))
   > Branch(Condition=@n8.ReturnValue)
   > n9 = Construct Object from Class(Class=&TA_GenericListItemData_C)
   n10 = Get CurrentActor()
   > Set integers(integers=Make Array("[0]"=1, "[1]"=@n8, "[2]"=Get CurrentActorCount(self=@n10) -> (CurrentActorCount=0)), self=@n9)
   > Set displayString(displayString=Get Class Display Name(Class=@n7), self=@n9)
   > addItem(self=Get TA_Generic01_List(), dataObject=@n9)
   PrintString(execute=@n6.LoopBody, InString=Get Class Display Name(Class=@n6))
   n11 = Get OtherClasses()
   n12 = Keys(execute=@n1.then_2, TargetMap=@n11)
   > n13 = Values!(execute, TargetMap:Map of Object Class References to Integers=@n11) -> (then, Values:Array of Integers)
   > n14 = For Each Loop(Array=@n12)
   n15 = "Get (a ref)"(Array=@n13, Dimension 1=@n14.Array Index)
   n16 = "Equal (Integer)"(A=@n14.Array Index)
   > Set l_otherTooltip(l_otherTooltip:String=Append!(
     A=Get l_otherTooltip() -> (l_otherTooltip:String)
     B=Select(Option 0:String="\r\n", Option 1:String, Index=@n16) -> (ReturnValue:String)
     C=Get Class Display Name(Class=@n14)
     D=" = "
     E:String="To String (Integer)"(InInt=@n15)
    ) -> (ReturnValue)) -> (Output_Get:String)
   n17 = Construct Object from Class(execute=@n1.then_3, Class=&TA_GenericListItemData_C)
   n18 = Get CurrentActor()
   > Set integers(
    integers=Make Array("[0]"=1, "[1]"=Get OtherCount() -> (OtherCount=0), "[2]"=Get CurrentActorCount(self=@n18) -> (CurrentActorCount=0))
    self=@n17
   )
   > Set displayString(displayString=Other, self=@n17)
   > Set strings!(
    execute
    strings:Array of Strings=Make Array!("[0]":String=Get l_otherTooltip()) -> (Array:Array of Strings)
    self:TA Generic List Item Data Object Reference=@n17
   ) -> (then, Output_Get:Array of Strings)
   > addItem(self=Get TA_Generic01_List(), dataObject=@n17)
   Get ClassList()
  ]}
 ValidateSelection {nodes [
   ValidateSelection() {is=title K2Node_FunctionEntry}
   > n1 = Sequence!(execute) -> (then_0, then_1)
   n2 = Get CurrentActor()
   > Is Valid(InputObject=@n2)
   > Unbind all Events from On Count Updated!(execute:Exec, self:EUA Actor Count Volume Object Reference=@n2) -> (then:Exec) {is=title K2Node_ClearDelegate}
   n3 = GetSelectionSet(execute=@n1.then_1, self=Get LevelEditorSubsystem())
   > n4 = GetSelectedObjects!(
    execute
    self:Typed Element Selection Set Object Reference=@n3
    InRequiredClass:Object Class Reference=&/EUA_ActorCountVolume.EUA_ActorCountVolume_C
   ) -> (then, ReturnValue:Array of Object References)
   > Branch(Condition=Is Valid Index(TargetArray:Array of Object References=@n4))
   > n5 = Cast To EUA_ActorCountVolume!(execute, Object="Get (a copy)"(Array:Array of Object References=@n4) -> (Output:Object Reference)) -> (then, CastFailed, AsEUA Actor Count Volume:EUA Actor Count Volume Object Reference) {is=title K2Node_DynamicCast}
   > UpdateInfoTarget(actor=@n5) {is=title K2Node_CallFunction}
   > Bind Event to On Count Updated(self:EUA Actor Count Volume Object Reference=@n5, Delegate=Create Event()) {is=title K2Node_AddDelegate}
  ]}
]
variables [
 ClassList {
  type="array<Class>"
  default [
   &BuildingProp
   &BuildingWall
   &BuildingFloor
   &BuildingRoof
   /Script/FortniteGame.BuildingDeco
   /Script/FortniteGame.BuildingCorner
   /Script/FortniteGame.BuildingPillar
   /Script/FortniteGame.BuildingStaticGeo
   /Script/FortniteGame.BuildingContainer
   /Script/LagerRuntime.FortAthenaLivingWorldStaticPointProvider
   /Script/FortniteGame.FortStaticMeshActor
  ]
 }
 ClassToCountMap {default {}}
 ClassToDepthMap {default {
   &Class'&BuildingContainer'=0
   &Class'&BuildingFloor'=0
   &Class'&BuildingProp'=0
   &Class'&BuildingRoof'=0
   &Class'&BuildingStairs'=0
   &Class'&BuildingStaticGeo'=0
   &Class'&BuildingWall'=0
  }}
 CurrentActor {type=EUA_ActorCountVolume_C, default=null}
 DisplayClassList {type="array<Class>", default []}
 OtherClasses {default {}}
 OtherCount {type=integer}
]
widgets {
 info {rootWidgetClass=&StackBox, widgetCount=6}
 widgets [
  StackBox_0 {widgetClassPath=&StackBox, widget=StackBox_0, properties {orientation=Orient_Vertical}}
  Border_0 {
   widgetClassPath=&Border
   parent=StackBox_0
   widget=Border_0
   properties {brushColor=<rgba 0 0 0 0>, padding=<ltrb 15 3 4 3>}
   slot {class=StackBoxSlot}
  }
  txtActor {
   widgetClassPath=&TextBlock
   bIsVariable=true
   parent=Border_0
   widget=txtActor
   properties {font {
     bForceMonospaced=false
     bMaterialIsStencil=false
     fontMaterial=null
     fontObject=&Roboto
     letterSpacing=0
     monospacedWidth=1
     size=10.5
     skewAmount=0
     typefaceFontName=Bold
     outlineSettings {
      bApplyOutlineToDropShadows=false
      bMiteredCorners=false
      bSeparateFillAlpha=false
      outlineMaterial=null
      outlineSize=0
      outlineColor=<rgba 0 0 0 1>
     }
    }}
   slot {class=BorderSlot, properties {padding=<ltrb 15 3 4 3>}}
  }
  Border_1 {
   widgetClassPath=&Border
   parent=StackBox_0
   widget=Border_1
   properties {brushColor=<rgba 0 0 0 0.25>, padding=<ltrb 4 4 4 4>}
   slot {class=StackBoxSlot, properties {size {sizeRule=Fill, value=1}}}
  }
  TA_Generic01_List {
   widgetClassPath=/Game/Blueprints/Blutilities/ForTechArt/Widgets/List/TA_Generic01_List
   bIsVariable=true
   parent=Border_1
   widget=TA_Generic01_List
   properties {
    allowNoSelection=true
    itemPadding=0
    listItemClass=/Game/NeverCook/EditorUtilities/PerfTools/ActorCountVolume/Widgets/ActorCountVolume_ListItem.ActorCountVolume_ListItem_C
   }
   slot {class=BorderSlot, properties {padding=<ltrb 4 4 4 4>}}
  }
  txtInfo {
   widgetClassPath=&TextBlock
   bIsVariable=true
   parent=StackBox_0
   widget=txtInfo
   properties {font {
     bForceMonospaced=false
     bMaterialIsStencil=false
     fontMaterial=null
     fontObject=&Roboto
     letterSpacing=0
     monospacedWidth=1
     size=12
     skewAmount=0
     typefaceFontName=Bold
     outlineSettings {
      bApplyOutlineToDropShadows=false
      bMiteredCorners=false
      bSeparateFillAlpha=false
      outlineMaterial=null
      outlineSize=0
      outlineColor=<rgba 0 0 0 1>
     }
    }}
   slot {class=StackBoxSlot}
  }
 ]
}
