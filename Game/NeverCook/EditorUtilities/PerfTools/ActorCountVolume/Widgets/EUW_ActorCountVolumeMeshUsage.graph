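# Editor utility widget for the Actor Count Volume perf tool that lists mesh usage counts and the Blueprints and locations behind them.
# setMeshData sorts rows by count through inline Python, fills the mesh list and totals it; a Max Count spinner hides rows above its value.
# Clicking a mesh fills Blueprint and location lists; locations focus the viewport.
# Blueprint clicks sync the Content Browser, or with Shift select and frame that class's level actors.
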EditorUtilityWidgetBlueprint /Game/NeverCook/EditorUtilities/PerfTools/ActorCountVolume/Widgets/EUW_ActorCountVolumeMeshUsage
graphs [
 EventGraph {nodes [
   n1 = "On Item Left Clicked (listLocations)"!() -> (OutputDelegate, then, isShiftDown:bool, isCtrlDown:bool, data:Object Reference, item:User Widget Object Reference) {is=title K2Node_ComponentBoundEvent}
   > HandleLocationClicked!(execute, self, Object:Object Reference=@n1.data) -> (then)
   n2 = "On Item Left Clicked (listMeshes)"!() -> (OutputDelegate, then, isShiftDown:bool, isCtrlDown:bool, data:Object Reference, item:User Widget Object Reference) {is=title K2Node_ComponentBoundEvent}
   > PopulateBPList!(execute, self, Object:Object Reference=@n2.data) -> (then)
   > PopulateLocationList!(execute, self, Object:Object Reference=@n2.data) -> (then)
   n3 = "On Item Left Clicked (listBP)"!() -> (OutputDelegate, then, isShiftDown:bool, isCtrlDown:bool, data:Object Reference, item:User Widget Object Reference) {is=title K2Node_ComponentBoundEvent}
   > HandleBPClicked!(execute, self, Object:Object Reference=@n3.data, ShiftHeld:bool=@n3.isShiftDown) -> (then)
   "On Clicked (btnResetSearch)"!() -> (OutputDelegate, then, widget:User Widget Object Reference, event:Pointer Event Structure) {is=title K2Node_ComponentBoundEvent}
   > ResetSearch()
   > RunSearch()
   n4 = "On Value Committed (spinBoxSearchCount)"!() -> (OutputDelegate, then, InValue:float, CommitMethod:ETextCommit Enum=Default) {is=title K2Node_ComponentBoundEvent}
   > n5 = Switch on ETextCommit!(execute, Selection:ETextCommit Enum=@n4.CommitMethod) -> (Default:Exec, OnEnter:Exec, OnUserMovedFocus:Exec, OnCleared:Exec) {is=title K2Node_SwitchEnum}
   RunSearch(execute=[@n5.OnEnter, @n5.OnUserMovedFocus])
   Event PreConstruct()
   Event Construct()
   Event Tick()
  ]}
 HandleBPClicked {nodes [
   n1 = HandleBPClicked!() -> (then, Object:Object Reference, ShiftHeld:bool=~) {is=title K2Node_FunctionEntry}
   > n2 = Cast To TA_GenericListItemData!(execute, Object=@n1) -> (then, CastFailed, AsTA Generic List Item Data:TA Generic List Item Data Object Reference) {is=title K2Node_DynamicCast}
   > Set l_data(l_data:TA Generic List Item Data Object Reference=@n2) -> (Output_Get:TA Generic List Item Data Object Reference)
   > Set l_assestPackageName(l_assestPackageName:String=Get displayString!(self:TA Generic List Item Data Object Reference=@n2) -> (displayString:String)) -> (Output_Get:String)
   > n3 = Branch(Condition=@n1.ShiftHeld)
   n4 = String To Name(InString=Get l_assestPackageName() -> (l_assestPackageName:String))
   > n5 = GetBlueprintAssets!(
    execute
    InFilter:"ARFilter Structure (by ref)"=MakeARFilter!(
     PackageNames:Array of Names=Make Array("[0]":Name=@n4) -> (Array:Array of Names)
     PackagePaths:Array of Names
     SoftObjectPaths:Array of Soft Object Path Structures
     ClassPaths:Array of Top Level Asset Path Structures
     RecursiveClassPathsExclusionSet:Set of Top Level Asset Path Structures
     ClassNames:Array of Names
     RecursiveClassesExclusionSet:Set of Names
     bRecursivePaths:bool=false
     bRecursiveClasses:bool=false
     bIncludeOnlyOnDiskAssets:bool=false
    ) -> (ReturnValue:ARFilter Structure)
   ) -> (then, OutAssetData:Array of Asset Data Structures)
   > Branch(Condition=Is Valid Index!(TargetArray:Array of Asset Data Structures=@n5, IndexToTest:int) -> (ReturnValue:bool))
   n6 = Break Asset Data(AssetData="Get (a ref)"(Array:Array of Asset Data Structures=@n5, Dimension 1=0) -> (Output:"Asset Data Structure (by ref)"))
   n7 = Append!(A="To String (Name)"(InName=@n6), B=., C:String="To String (Name)"(InName=@n6.AssetName), D:String=_C) -> (ReturnValue)
   > n8 = LoadClassAsset_Blocking(AssetClass=To Soft Class Reference(SoftClassPath=MakeSoftClassPath(PathString=@n7)))
   > n9 = Cast To Actor Class!(execute:Exec, Class:Object Class Reference=@n8) -> (then:Exec, CastFailed:Exec, AsActor:Actor Class Reference) {is=title K2Node_ClassDynamicCast}
   > n10 = GetAllActorsOfClass!(execute, WorldContextObject:Object Reference, ActorClass:Actor Class Reference=@n9) -> (then, OutActors:Array of Actor Object References)
   > Focus Actors in Viewport!(" ":Exec, actors:Array of Actor Object References=@n10, collidableOnly:bool, DoCameraMove:bool=true) -> ("  ":Exec, Loc:Vector=~, Rot:Rotator=~) {is=title K2Node_MacroInstance}
   > SetSelectedLevelActors(self=Get EditorActorSubsystem(), ActorsToSelect=@n10)
   SyncBrowserToObjects(execute=@n3.else, AssetPaths=Make Array("[0]":String=Get l_assestPackageName()) -> (Array:Array of Strings))
   BreakTopLevelAssetPath!(TopLevelAssetPath:"Top Level Asset Path Structure (by ref)"=@n6.AssetClassPath) -> (PathString:String)
  ]}
 HandleLocationClicked {nodes [
   n1 = HandleLocationClicked!() -> (then, Object:Object Reference) {is=title K2Node_FunctionEntry}
   > n2 = Cast To TA_GenericListItemData(Object=@n1)
   > Focus Bounds in Viewport!(
    " ":Exec
    center:Vector="Get (a copy)"(Array:Array of Vectors=Get vectors!(self:TA Generic List Item Data Object Reference=@n2) -> (vectors:Array of Vectors)) -> (Output:Vector)
    bounds:Vector="200,200,200"
   ) -> ("  ":Exec) {is=title K2Node_MacroInstance}
  ]}
 PopulateBPList {nodes [
   n1 = PopulateBPList!() -> (then, Object:Object Reference) {is=title K2Node_FunctionEntry}
   n2 = Get listBP() -> (listBP:TA Generic 01 List Object Reference)
   > clearItems(self:TA Generic 01 List Object Reference=@n2)
   > n3 = Cast To TA_GenericListItemData(Object=@n1)
   > n4 = For Each Loop(Array=Get strings!(self:TA Generic List Item Data Object Reference=@n3) -> (strings:Array of Strings))
   > addItemGeneric!(execute, self:TA Generic 01 List Object Reference=@n2, displayString:String=@n4) -> (then, widget:User Widget Object Reference, data:Object Reference)
  ]}
 PopulateLocationList {nodes [
   n1 = PopulateLocationList!() -> (then, Object:Object Reference) {is=title K2Node_FunctionEntry}
   > n2 = Cast To TA_GenericListItemData(Object=@n1)
   > clearItems(self=Get listLocations() -> (listLocations:TA Generic 01 List Object Reference))
   > n3 = For Each Loop(Array:Array of Vectors=Get vectors(self=@n2)) -> (Array Element:Vector)
   > Set l_locationElement(l_locationElement:Vector=@n3) -> (Output_Get:Vector)
   > n4 = Construct Object from Class(Class=/Game/Blueprints/Blutilities/ForTechArt/Data/TA_GenericListItemData.TA_GenericListItemData_C) -> (ReturnValue:TA Generic List Item Data Object Reference)
   > Set displayString!(
    execute
    displayString:String="To String (Vector)"!(InVec:Vector=Get l_locationElement() -> (l_locationElement:Vector)) -> (ReturnValue:String)
    self:TA Generic List Item Data Object Reference=@n4
   ) -> (then, Output_Get:String)
   > Set vectors!(
    execute
    vectors:Array of Vectors=Make Array("[0]":Vector=Get l_locationElement()) -> (Array:Array of Vectors)
    self:TA Generic List Item Data Object Reference=@n4
   ) -> (then, Output_Get:Array of Vectors)
   > addItem!(execute, self:TA Generic 01 List Object Reference=Get listLocations(), dataObject:Object Reference=@n4) -> (then, newItem:User Widget Object Reference, newData:Object Reference)
  ]}
 ResetSearch {nodes [
   ResetSearch() {is=title K2Node_FunctionEntry}
   > Set Value!(
    execute
    Value:float=999
    self:Spin Box Object Reference=Get spinBoxSearchCount() -> (spinBoxSearchCount:Editor Utility Spin Box Object Reference)
   ) -> (then, Output_Get:float)
  ]}
 RunSearch {nodes [
   RunSearch() {is=title K2Node_FunctionEntry}
   n1 = Get spinBoxSearchCount()
   > Set l_max(l_max:int=Truncate(A=Get Value!(self:Spin Box Object Reference=@n1) -> (Value:float))) -> (Output_Get:int=0)
   > n2 = getListItems!(execute, self:TA Generic 01 List Object Reference=Get listMeshes() -> (listMeshes:TA Generic 01 List Object Reference)) -> (then, items:Array of User Widget Object References)
   > n3 = For Each Loop(Array:Array of User Widget Object References=@n2) -> (Array Element:User Widget Object Reference)
   > n4 = Cast To ActorCountVolumeMeshUsage_ListItem!(execute, Object=@n3) -> (then, CastFailed, AsActor Count Volume Mesh Usage List Item:Actor Count Volume Mesh Usage List Item Object Reference) {is=title K2Node_DynamicCast}
   > n5 = Cast To TA_GenericListItemData(Object=Get currentListItemData!(self:TA Base Editor Button Widget Object Reference=@n4) -> (currentListItemData:Object Reference))
   n6 = "Get (a copy)"(Array:Array of Integers=Get integers!(self:TA Generic List Item Data Object Reference=@n5) -> (integers:Array of Integers)) -> (Output:int)
   n7 = Get l_max() -> (l_max:int=0)
   > SetVisibility(self=@n4, InVisibility=Select(Option 0=Collapsed, Option 1=Visible, Index="integer <= integer"(A=@n6, B=@n7)))
  ]}
 setMeshData {nodes [
   n1 = Get l_countArrayToSort() -> (l_countArrayToSort:Array of Integers)
   n2 = Execute Python Script!(
    execute=@n9.Completed
    PythonScript="import unreal\r\n\r\nsorted_dict = {}\r\nfor i in range(len(import_array)):\r\n    sorted_dict[i] = import_array[i]\r\n\r\nif len(sorted_dict) > 0:\r\n    sorted_dict = sorted(sorted_dict.items(), key=lambda item: item[1])"
    import_array:Array of Integers=@n1
   ) -> (then, ReturnValue, sorted_dict:Map of Integers to Integers) {inputs [import_array], outputs [sorted_dict]}
   > n3 = Set l_sortedMap(l_sortedMap:Map of Integers to Integers=@n2.sorted_dict) -> (Output_Get:Map of Integers to Integers)
   n4 = Values!(execute, TargetMap:Map of Integers to Integers=@n3) -> (then, Values:Array of Integers)
   n5 = Keys!(execute=@n3.then, TargetMap:Map of Integers to Integers=@n3) -> (then, Keys:Array of Integers)
   n6 = For Each Loop(Exec=[@n4.then, @n5.then], Array:Array of Integers=@n5) -> (Array Element:int)
   > Add(
    TargetArray:Array of Actor Count Volume Mesh Table Struct Structures=Get l_sortedMeshData() -> (l_sortedMeshData:Array of Actor Count Volume Mesh Table Struct Structures)
    NewItem:"Actor Count Volume Mesh Table Struct Structure (by ref)"="Get (a ref)"(
     Array:Array of Actor Count Volume Mesh Table Struct Structures=Get CurrentMeshData() -> (CurrentMeshData:Array of Actor Count Volume Mesh Table Struct Structures)
     Dimension 1=@n6
    ) -> (Output:"Actor Count Volume Mesh Table Struct Structure (by ref)")
   ) {is=title K2Node_CallArrayFunction}
   n7 = setMeshData!() -> (then, MeshData:Array of Actor Count Volume Mesh Table Struct Structures) {is=title K2Node_FunctionEntry}
   > Set CurrentMeshData(CurrentMeshData:Array of Actor Count Volume Mesh Table Struct Structures=@n7) -> (Output_Get:Array of Actor Count Volume Mesh Table Struct Structures)
   > clearItems(self=Get listMeshes())
   > n8 = Sequence()
   > n9 = For Each Loop(Array:Array of Actor Count Volume Mesh Table Struct Structures=Get CurrentMeshData()) -> (Array Element:Actor Count Volume Mesh Table Struct Structure)
   > Add(
    TargetArray:Array of Integers=@n1
    NewItem:"Integer (by ref)"=Break Actor Count Volume Mesh Table Struct!(ActorCountVolumeMeshTableStruct:"Actor Count Volume Mesh Table Struct Structure (by ref)"=@n9) -> (
     AssetName_15_#1:Name
     Count_16_#2:int
     AssetRefs_23_#3:Array of Strings
     MaxCount_27_#4:int
     Locations_31_#5:Array of Vectors
    ) {is=title K2Node_BreakStruct}.Count_16_#2
   )
   n10 = Set CurrentMeshData(execute=@n8.then_1, CurrentMeshData=Get l_sortedMeshData())
   > n11 = For Each Loop(Array=@n10)
   RunSearch(execute=@n11.Completed) {is=title K2Node_CallFunction}
   > "SetText (Text)"(
    self=Get txtCount() -> (txtCount:Text Object Reference)
    InText=Format Text!(Format="LOCTEXT(\"Mesh Count\\r\\n{Total}\")", Total:int=Get l_totalCount() -> (l_totalCount:int=0)) -> (Result)
   )
   n12 = Sequence(execute=@n11.LoopBody)
   > n13 = Construct Object from Class(Class=&TA_GenericListItemData_C)
   n14 = Break Actor Count Volume Mesh Table Struct(ActorCountVolumeMeshTableStruct=@n11)
   > Set strings!(execute, strings:Array of Strings=@n14.AssetRefs_23_#3, self:TA Generic List Item Data Object Reference=@n13) -> (then, Output_Get:Array of Strings)
   > Set displayString(displayString="To String (Name)"(InName=@n14), self=@n13)
   > Set integers!(
    execute
    integers:Array of Integers=Make Array!("[0]":int=@n14.Count_16_#2, "[1]":int=@n14.MaxCount_27_#4) -> (Array:Array of Integers)
    self:TA Generic List Item Data Object Reference=@n13
   ) -> (then, Output_Get:Array of Integers)
   > Set l_totalCount(l_totalCount:int=int + int(A=@n14.Count_16_#2, B=Get l_totalCount() -> (l_totalCount=0))) -> (Output_Get:int=0)
   > Set vectors(vectors=@n14.Locations_31_#5, self=@n13)
   addItem(execute=@n12.then_1, self=Get listMeshes(), dataObject=@n13)
  ]}
]
variables [CurrentMeshData {type="array<ActorCountVolumeMeshTableStruct>", default []}]
widgets {
 info {rootWidgetClass=&VerticalBox, widgetCount=20}
 widgets [
  VerticalBox_0 {widgetClassPath=&VerticalBox, widget=VerticalBox_0}
  Border_0 {
   widgetClassPath=&Border
   parent=VerticalBox_0
   widget=Border_0
   properties {brushColor=<rgba 0 0 0 0.25>}
   slot {class=VerticalBoxSlot}
  }
  HorizontalBox_0 {widgetClassPath=&HorizontalBox, parent=Border_0, widget=HorizontalBox_0, slot {class=BorderSlot}}
  btnResetSearch {
   widgetClassPath=&TA_Generic01_Button
   parent=HorizontalBox_0
   widget=btnResetSearch
   properties {label=Reset, buttonColor_Default=<rgba 0.4271 0.2288 0.0097 1>}
   slot {class=HorizontalBoxSlot, properties {verticalAlignment=VAlign_Center}}
  }
  CommonTextBlock_52 {
   widgetClassPath=&CommonTextBlock
   parent=HorizontalBox_0
   widget=CommonTextBlock_52
   properties {
    text=Max Count:
    font {
     bForceMonospaced=false
     bMaterialIsStencil=false
     fontMaterial=null
     fontObject=&NotoSans
     letterSpacing=0
     monospacedWidth=1
     size=12
     skewAmount=0
     typefaceFontName=NotoSans-Regular
     outlineSettings {
      bApplyOutlineToDropShadows=false
      bMiteredCorners=false
      bSeparateFillAlpha=false
      outlineMaterial=null
      outlineSize=0
      outlineColor=<rgba 0 0 0 1>
     }
    }
    shadowOffset=<0 0>
   }
   slot {class=HorizontalBoxSlot, properties {verticalAlignment=VAlign_Center}}
  }
  spinBoxSearchCount {
   widgetClassPath=/Script/Blutility.EditorUtilitySpinBox
   bIsVariable=true
   parent=HorizontalBox_0
   widget=spinBoxSearchCount
   properties {
    bOverride_MaxSliderValue=true
    bOverride_MinSliderValue=true
    bOverride_MinValue=true
    maxSliderValue=999
    minDesiredWidth=100
    minSliderValue=1
    minValue=1
    value=999
    foregroundColor {colorUseRule=UseColor_Specified, specifiedColor=<rgba 1 1 1 1>}
   }
   slot {class=HorizontalBoxSlot}
  }
  StackBox_0 {
   widgetClassPath=&StackBox
   parent=VerticalBox_0
   widget=StackBox_0
   slot {class=VerticalBoxSlot, properties {size {sizeRule=Fill, value=1}}}
  }
  VerticalBox_1 {
   widgetClassPath=&VerticalBox
   parent=StackBox_0
   widget=VerticalBox_1
   slot {class=StackBoxSlot, properties {size {sizeRule=Fill, value=1}}}
  }
  borderMain {
   widgetClassPath=&Border
   bIsVariable=true
   parent=VerticalBox_1
   widget=borderMain
   properties {brushColor=<rgba 0 0 0 0.6087>, padding=<ltrb 0 0 3 0>}
   slot {class=VerticalBoxSlot}
  }
  HorizontalBox {
   widgetClassPath=&HorizontalBox
   parent=borderMain
   widget=HorizontalBox
   slot {class=BorderSlot, properties {padding=<ltrb 0 0 3 0>}}
  }
  imgIcon {
   widgetClassPath=&Image
   bIsVariable=true
   parent=HorizontalBox
   widget=imgIcon
   properties {
    visibility=Collapsed
    brush {
     bIsDynamicallyLoaded=false
     drawAs=Image
     imageType=NoImage
     mirroring=NoMirror
     resourceName=None
     resourceObject=/Game/UI/Fort_Traits/Circle
     tiling=NoTile
     imageSize=<12 12>
     margin=<ltrb 0 0 0 0>
     outlineSettings {
      bUseBrushTransparency=false
      roundingType=HalfHeightRadius
      width=0
      color {colorUseRule=UseColor_Specified, specifiedColor=<rgba 0 0 0 0>}
      cornerRadii=<0 0 0 0>
     }
     tintColor {colorUseRule=UseColor_Specified, specifiedColor=<rgba 1 1 1 1>}
     uVRegion {bIsValid=false, max=<0 0>, min=<0 0>}
    }
   }
   slot {class=HorizontalBoxSlot, properties {horizontalAlignment=HAlign_Center, verticalAlignment=VAlign_Center}}
  }
  txtCount {
   widgetClassPath=&TextBlock
   bIsVariable=true
   parent=HorizontalBox
   widget=txtCount
   properties {
    clipping=ClipToBounds
    justification=Center
    minDesiredWidth=45
    text=Mesh Count
    toolTipText=Number of references to this mesh
    colorAndOpacity {colorUseRule=UseColor_Specified, specifiedColor=<rgba 0.8646 0.8646 0.8646 1>}
    font {
     bForceMonospaced=false
     bMaterialIsStencil=false
     fontMaterial=null
     fontObject=&Roboto
     letterSpacing=0
     monospacedWidth=1
     size=6
     skewAmount=0
     typefaceFontName=Bold
     outlineSettings {
      bApplyOutlineToDropShadows=false
      bMiteredCorners=false
      bSeparateFillAlpha=false
      outlineMaterial=null
      outlineSize=0
      outlineColor=<rgba 0 0 0 1>
     }
    }
    shadowColorAndOpacity=<rgba 0 0 0 0.2>
   }
   slot {class=HorizontalBoxSlot, properties {verticalAlignment=VAlign_Center}}
  }
  ScaleBox_0 {
   widgetClassPath=&ScaleBox
   parent=HorizontalBox
   widget=ScaleBox_0
   properties {clipping=ClipToBounds, stretchDirection=DownOnly}
   slot {class=HorizontalBoxSlot, properties {padding=<ltrb 0 5 0 5>, size {sizeRule=Fill, value=1}}}
  }
  txtMain {
   widgetClassPath=&TextBlock
   bIsVariable=true
   parent=ScaleBox_0
   widget=txtMain
   properties {
    clipping=ClipToBounds
    justification=Center
    text=Mesh Package Name
    colorAndOpacity {colorUseRule=UseColor_Specified, specifiedColor=<rgba 0.8646 0.8646 0.8646 1>}
    font {
     bForceMonospaced=false
     bMaterialIsStencil=false
     fontMaterial=null
     fontObject=&Roboto
     letterSpacing=0
     monospacedWidth=1
     size=6
     skewAmount=0
     typefaceFontName=Bold
     outlineSettings {
      bApplyOutlineToDropShadows=false
      bMiteredCorners=false
      bSeparateFillAlpha=false
      outlineMaterial=null
      outlineSize=0
      outlineColor=<rgba 0 0 0 1>
     }
    }
    shadowColorAndOpacity=<rgba 0 0 0 0.2>
   }
   slot {class=ScaleBoxSlot, properties {horizontalAlignment=HAlign_Fill}}
  }
  imageEdge_2 {
   widgetClassPath=&Image
   bIsVariable=true
   parent=HorizontalBox
   widget=imageEdge_2
   properties {
    brush {
     bIsDynamicallyLoaded=false
     drawAs=Image
     imageType=NoImage
     mirroring=NoMirror
     resourceName=None
     resourceObject=null
     tiling=NoTile
     imageSize=<1 0>
     margin=<ltrb 0 0 0 0>
     outlineSettings {
      bUseBrushTransparency=false
      roundingType=HalfHeightRadius
      width=0
      color {colorUseRule=UseColor_Specified, specifiedColor=<rgba 0 0 0 0>}
      cornerRadii=<0 0 0 0>
     }
     tintColor {colorUseRule=UseColor_Specified, specifiedColor=<rgba 1 1 1 1>}
     uVRegion {bIsValid=false, max=<0 0>, min=<0 0>}
    }
    colorAndOpacity=<rgba 0.3542 0.3542 0.3542 1>
   }
   slot {class=HorizontalBoxSlot, properties {horizontalAlignment=HAlign_Right, padding=<ltrb 3 0 3 0>}}
  }
  imageEdge_1 {
   widgetClassPath=&Image
   bIsVariable=true
   parent=HorizontalBox
   widget=imageEdge_1
   properties {
    brush {
     bIsDynamicallyLoaded=false
     drawAs=Image
     imageType=NoImage
     mirroring=NoMirror
     resourceName=None
     resourceObject=null
     tiling=NoTile
     imageSize=<1 0>
     margin=<ltrb 0 0 0 0>
     outlineSettings {
      bUseBrushTransparency=false
      roundingType=HalfHeightRadius
      width=0
      color {colorUseRule=UseColor_Specified, specifiedColor=<rgba 0 0 0 0>}
      cornerRadii=<0 0 0 0>
     }
     tintColor {colorUseRule=UseColor_Specified, specifiedColor=<rgba 1 1 1 1>}
     uVRegion {bIsValid=false, max=<0 0>, min=<0 0>}
    }
    colorAndOpacity=<rgba 0.3542 0.3542 0.3542 1>
   }
   slot {class=HorizontalBoxSlot, properties {padding=<ltrb 3 0 3 0>}}
  }
  txtCountBP {
   widgetClassPath=&TextBlock
   bIsVariable=true
   parent=HorizontalBox
   widget=txtCountBP
   properties {
    clipping=ClipToBounds
    justification=Center
    minDesiredWidth=63
    text=BP Count
    toolTipText=Num of blueprints using this mesh in the area.
    colorAndOpacity {colorUseRule=UseColor_Specified, specifiedColor=<rgba 0.8646 0.8646 0.8646 1>}
    font {
     bForceMonospaced=false
     bMaterialIsStencil=false
     fontMaterial=null
     fontObject=&Roboto
     letterSpacing=0
     monospacedWidth=1
     size=6
     skewAmount=0
     typefaceFontName=Bold
     outlineSettings {
      bApplyOutlineToDropShadows=false
      bMiteredCorners=false
      bSeparateFillAlpha=false
      outlineMaterial=null
      outlineSize=0
      outlineColor=<rgba 0 0 0 1>
     }
    }
    shadowColorAndOpacity=<rgba 0 0 0 0.2>
   }
   slot {class=HorizontalBoxSlot, properties {verticalAlignment=VAlign_Center}}
  }
  listMeshes {
   widgetClassPath=/Game/Blueprints/Blutilities/ForTechArt/Widgets/List/TA_Generic01_List
   bIsVariable=true
   parent=VerticalBox_1
   widget=listMeshes
   properties {listItemClass=&/ActorCountVolumeMeshUsage_ListItem.ActorCountVolumeMeshUsage_ListItem_C}
   slot {class=VerticalBoxSlot, properties {size {sizeRule=Fill, value=1}}}
  }
  listBP {
   widgetClassPath=&TA_Generic01_List
   bIsVariable=true
   parent=StackBox_0
   widget=listBP
   slot {class=StackBoxSlot, properties {size {sizeRule=Fill, value=1}}}
  }
  listLocations {
   widgetClassPath=&TA_Generic01_List
   bIsVariable=true
   parent=StackBox_0
   widget=listLocations
   slot {class=StackBoxSlot, properties {size {sizeRule=Fill, value=1}}}
  }
 ]
}
