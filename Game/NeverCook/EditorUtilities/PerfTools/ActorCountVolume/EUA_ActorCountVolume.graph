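# Editor utility volume that counts World Partition actor descriptors in its tall column for perf auditing.
# Descriptors are filtered by type, size, spatial loading, class, building-prop collection and data layer visibility.
# Construction refreshes the count, writes it to the cube's custom primitive data and broadcasts OnCountUpdated.
# GetMeshUsage tallies colliding static meshes via Blueprint asset-registry dependencies, fills a developer-folder DataTable through CSV and opens a usage widget.
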EditorUtilityBlueprint /Game/NeverCook/EditorUtilities/PerfTools/ActorCountVolume/EUA_ActorCountVolume
graphs [
 EventGraph {nodes [
   Event BeginPlay()
   Event ActorBeginOverlap()
   Event Tick!() -> (OutputDelegate, then, DeltaSeconds:float)
  ]}
 FilterActorDesc {nodes [
   n1 = FilterActorDesc!() -> (then, ActorDesc:"Actor Desc Structure (by ref)", useClassFilter:bool=~) {is=title K2Node_FunctionEntry}
   > Set l_desc(l_desc:Actor Desc Structure=@n1) -> (Output_Get:Actor Desc Structure)
   > n2 = Sequence!(execute) -> (then_0, then_1, then_2:Exec, then_3:Exec, then_4:Exec, then_5:Exec, then_6:Exec, then_7:Exec)
   n3 = Break Actor Desc!(ActorDesc:"Actor Desc Structure (by ref)"=Get l_desc() -> (l_desc:Actor Desc Structure)) -> (NativeClass:Object Class Reference, Label:Name, Bounds:Box Structure, bIsSpatiallyLoaded:bool=~, bActorIsEditorOnly:bool=~, FolderPath:Name) {is=title K2Node_BreakStruct}
   > n4 = Cast To WorldPartitionHLOD Class!(execute:Exec, Class:Object Class Reference=@n3) -> (then:Exec, CastFailed:Exec, AsWorld Partition HLOD:World Partition HLOD Class Reference) {is=title K2Node_ClassDynamicCast}
   > Return Node!(execute, Valid:bool=false)
   Cast To EditorUtilityActor Class!(execute:Exec=@n4.CastFailed, Class:Object Class Reference=@n3) -> (then:Exec, CastFailed:Exec, AsEditor Utility Actor:Editor Utility Actor Class Reference) {is=title K2Node_ClassDynamicCast}
   > Return Node(Valid=false)
   Branch(execute=@n2.then_1, Condition=Contains(SearchIn="To String (Name)"(InName=@n3.Label), Substring=landscape))
   > Return Node(Valid=false)
   n5 = GetBoxCenterSize!(Box:Box Structure=@n3.Bounds) -> (Center:Vector, Dimensions:Vector)
   Branch(execute=@n2.then_2, Condition="float > float"(A=Vector Length(A=@n5.Dimensions), B=50000))
   > Return Node(Valid=false)
   n6 = Branch(execute=@n2.then_3, Condition=@n3.bIsSpatiallyLoaded)
   Return Node(execute=@n6.else, Valid=false)
   n7 = Get IncludeEditorOnly() -> (IncludeEditorOnly:bool)
   Branch(execute=@n2.then_4, Condition=AND Boolean(A=NOT Boolean(A=@n7), B=@n3.bActorIsEditorOnly))
   > Return Node(Valid=false)
   Branch(execute=@n2.then_5, Condition=@n1.useClassFilter)
   > n8 = For Each Loop(Array:Array of Object Class References=Get TransientClassFilter() -> (TransientClassFilter:Array of Object Class References)) -> (Array Element:Object Class Reference)
   > n9 = FilterActorDescByClass!(execute, self, ActorDesc:"Actor Desc Structure (by ref)"=Get l_desc(), ParentClass:Object Class Reference=@n8) -> (then, IsValid:bool)
   > n10 = Branch(Condition=@n9)
   Return Node(execute=@n10.else, Valid=false)
   n11 = Branch(execute=@n2.then_6, Condition=Get IncludeSDP() -> (IncludeSDP:bool))
   n12 = Get l_desc()
   n13 = Split!(
    execute:Exec=@n11.else
    SourceString=BreakSoftObjectPath(InSoftObjectPath=Break Actor Desc!(ActorDesc=@n12) -> (Class:Soft Object Path Structure, ActorPackage:Name, FolderPath))
    InStr=.
    SearchCase
    SearchDir=FromStart
   ) -> (then:Exec, LeftS, RightS, ReturnValue)
   > Branch(Condition=@n13.ReturnValue)
   > Branch(Condition=Contains Item(
     TargetArray:Array of Strings=Get CachedSDPCollectionNames() -> (CachedSDPCollectionNames:Array of Strings)
     ItemToFind:"String (by ref)"=@n13
    ))
   > Return Node(Valid=false)
   Return Node(execute=@n2.then_7)
  ]}
 FilterActorDescByClass {nodes [
   n1 = FilterActorDescByClass!() -> (then, ActorDesc:"Actor Desc Structure (by ref)", ParentClass:Object Class Reference) {is=title K2Node_FunctionEntry}
   > n2 = IsDescOfClass!(execute, __WorldContext:Object Reference, ActorDesc:Actor Desc Structure=@n1, ParentClass:Object Class Reference=@n1.ParentClass) -> (then, IsValid:bool)
   > Return Node!(execute, IsValid:bool=@n2)
   Set l_desc(l_desc=@n1)
   > Set l_parentClass(l_parentClass:Object Class Reference=@n1.ParentClass) -> (Output_Get:Object Class Reference)
   > n3 = Sequence!(execute) -> (then_0, then_1, then_2)
   n4 = Break Actor Desc!(ActorDesc=Get l_desc()) -> (NativeClass:Object Class Reference, Class, ActorPackage, FolderPath)
   n5 = Get l_parentClass() -> (l_parentClass:Object Class Reference)
   > Branch(Condition=OR Boolean(A=ClassIsChildOf(TestClass=@n4, ParentClass=@n5), B="Equal (Class)"(A=@n4)))
   > Return Node(IsValid=true)
   n6 = BreakSoftObjectPath(InSoftObjectPath=Break Actor Desc!(ActorDesc=Get l_desc()) -> (Class, FolderPath))
   n7 = LoadClassAsset_Blocking(execute=@n3.then_1, AssetClass=To Soft Class Reference(SoftClassPath=MakeSoftClassPath(PathString=@n6)))
   > n8 = Branch(Condition="Equal (Class)"(A=@n7, B=Get l_parentClass()))
   Return Node(execute=[@n9.then, @n8.then], IsValid=true)
   n9 = Branch(execute=@n8.else, Condition=ClassIsChildOf(TestClass=@n7, ParentClass=Get l_parentClass()))
   Return Node(execute=@n3.then_2, IsValid=false)
  ]}
 GetActorCount {nodes [
   n1 = GetActorCount!() -> (then, AllowClassFiltering:bool=~) {is=title K2Node_FunctionEntry}
   > Clear(TargetArray:Array of Actor Desc Structures=Get CurrentActorDescs() -> (CurrentActorDescs:Array of Actor Desc Structures))
   > n2 = Sequence()
   > Clear(TargetArray:Array of Strings=Get CachedSDPCollectionNames())
   > n3 = Branch(Condition=Get IncludeSDP())
   n4 = Get CollectionManagerScriptingSubsystem()
   n5 = GetBaseGameCollectionContainer!(self:Collection Manager Subsystem Object Reference=@n4) -> (ReturnValue:Collection Container Source Structure)
   n6 = GetAssetsInCollection!(
    execute=@n3.else
    self:Collection Manager Subsystem Object Reference=@n4
    Collection:"Collection Structure (by ref)"=Make Collection!(
     Container:Name=Break Collection Container Source!(CollectionScriptingContainerSource:"Collection Container Source Structure (by ref)"=@n5) -> (Name:Name, Title:Text) {is=title K2Node_BreakStruct}
     Name:Name=FN_BR_SimpleBuildingProps
     ShareType:ECollectionScriptingShareType Enum=Shared
    ) -> (CollectionScriptingRef:Collection Structure) {is=title K2Node_MakeStruct}
   ) -> (then, OutAssets:Array of Asset Data Structures, ReturnValue:bool)
   > n7 = For Each Loop(Array:Array of Asset Data Structures=@n6) -> (Array Element:Asset Data Structure)
   > Add(TargetArray=Get CachedSDPCollectionNames(), NewItem="To String (Name)"(InName=Break Asset Data(AssetData=@n7))) {is=title K2Node_CallArrayFunction}
   n8 = GetActorDescriptions!(execute=@n2.then_1, self, useClassFilter:bool=@n1) -> (then, count:int=0, descs:Array of Actor Desc Structures)
   > Set l_actorCount(l_actorCount:int=@n8) -> (Output_Get:int=0)
   SetActorCount!(execute=@n2.then_2, self, CurrentActorCount:int=Get l_actorCount() -> (l_actorCount:int=0)) -> (then)
   > Call On Count Updated!(execute:Exec, self:EUA Actor Count Volume Object Reference, actor:EUA Actor Count Volume Object Reference=Self-Reference()) -> (then:Exec) {is=title K2Node_CallDelegate}
  ]}
 GetActorDescriptions {nodes [
   n1 = GetActorDescriptions!() -> (then, useClassFilter:bool=~) {is=title K2Node_FunctionEntry}
   > n2 = Sequence!(execute) -> (then_0, then_1)
   n3 = Get Actor Bounds(bOnlyCollidingComponents=false)
   n4 = vector + vector(A=@n3.BoxExtent, B="0,0,999999")
   > n5 = GetIntersectingActorDescs!(
    execute
    InBox:"Box Structure (by ref)"=MakeBox!(Min:Vector=vector - vector(A=@n3, B=@n4), Max:Vector=vector + vector(A=@n3, B=@n4)) -> (ReturnValue:Box Structure)
   ) -> (then, OutActorDescs:Array of Actor Desc Structures, ReturnValue:bool)
   > Branch(Condition=@n5.ReturnValue)
   > n6 = For Each Loop(Array:Array of Actor Desc Structures=@n5) -> (Array Element:Actor Desc Structure)
   > Branch(Condition=FilterActorDesc!(ActorDesc:"Actor Desc Structure (by ref)"=@n6, useClassFilter:bool=@n1) -> (Valid:bool) {is=title K2Node_CallFunction})
   > n7 = Set l_DLPassthrough(l_DLPassthrough:Array of Soft Object Path Structures=Break Actor Desc!(ActorDesc=@n6) -> (
     Guid:Guid Structure
     NativeClass:Object Class Reference
     Class
     Name:Name
     Label:Name
     Bounds:Box Structure
     RuntimeGrid:Name
     bIsSpatiallyLoaded:bool=~
     bActorIsEditorOnly:bool=~
     ActorPackage:Name
     ActorPath:Name
     FolderPath
     DataLayerAssets:Array of Soft Object Path Structures
    ).DataLayerAssets) -> (Output_Get:Array of Soft Object Path Structures)
   > n8 = TestDataLayers!(execute, self, DataLayers:Array of Soft Object Path Structures=@n7) -> (then, valid:bool)
   > Branch(Condition=@n8)
   > Increment Int(Value=Get l_actorCount() -> (l_actorCount=0))
   > Add(
    TargetArray:Array of Actor Desc Structures=Get l_descs() -> (l_descs:Array of Actor Desc Structures)
    NewItem:"Actor Desc Structure (by ref)"=@n6
   )
   Return Node!(execute=@n2.then_1, count:int=Get l_actorCount() -> (l_actorCount=0), descs:Array of Actor Desc Structures=Get l_descs())
  ]}
 GetMeshUsage {nodes [
   GetMeshUsage() {is=title K2Node_FunctionEntry}
   > Set l_csv(l_csv:String="---,Asset Name,Count,AssetRefs,MaxCount,Locations\r\n") -> (Output_Get:String)
   > n1 = Sequence!(execute) -> (then_0, then_1, then_2:Exec, then_3:Exec, then_4:Exec, then_5:Exec)
   > n2 = GetActorDescriptions(useClassFilter=false) -> (count=0) {is=title K2Node_CallFunction}
   > n3 = For Each Loop(Array=@n2.descs)
   n4 = Break Actor Desc(ActorDesc=@n3) -> (bIsSpatiallyLoaded=~, bActorIsEditorOnly=~)
   > n5 = Branch(Condition=@n4.bActorIsEditorOnly)
   n6 = Set l_class(execute=@n5.else, l_class:Soft Object Path Structure=@n4.Class) -> (Output_Get:Soft Object Path Structure)
   > n7 = Sequence!(execute) -> (then_0, then_1)
   n8 = Find(
    TargetMap:Map of Soft Object Path Structures to Integers=Get l_classSet() -> (l_classSet:Map of Soft Object Path Structures to Integers)
    Key:"Soft Object Path Structure (by ref)"=@n6
   ) -> (Value:int)
   > n9 = Branch(Condition=@n8.ReturnValue)
   > Add!(
    execute
    TargetMap:Map of Soft Object Path Structures to Integers=Get l_classSet()
    Key:"Soft Object Path Structure (by ref)"=Get l_class() -> (l_class:Soft Object Path Structure)
    Value:"Integer (by ref)"=int + int(A=@n8)
   ) -> (then) {is=title K2Node_CallFunction}
   Add(execute=@n9.else, TargetMap=Get l_classSet(), Key=Get l_class(), Value=1)
   n10 = BreakSoftObjectPath(InSoftObjectPath=Get l_class())
   Set l_classRootPath(
    execute=@n7.then_1
    l_classRootPath:Name=String To Name(InString=Split!(SourceString=@n10, InStr=., SearchCase, SearchDir=FromStart) -> (LeftS, RightS, ReturnValue))
   ) -> (Output_Get:Name=None)
   > Set l_actorCenterLocation(l_actorCenterLocation:Vector="Get Center (Box)"!(InBox:"Box Structure (by ref)"=@n4.Bounds) -> (ReturnValue:Vector) {is=title K2Node_CallFunction}) -> (Output_Get:Vector)
   n11 = Find(
    TargetMap:Map of Names to Actor Count Volume Mesh Location Struct Structures=Get l_classLocationSet() -> (l_classLocationSet:Map of Names to Actor Count Volume Mesh Location Struct Structures)
    Key:"Name (by ref)"=Get l_classRootPath() -> (l_classRootPath:Name=None)
   ) -> (Value:Actor Count Volume Mesh Location Struct Structure)
   > n12 = Branch(Condition=@n11.ReturnValue)
   > n13 = Set l_locationEditStruct(l_locationEditStruct:Actor Count Volume Mesh Location Struct Structure=@n11) -> (Output_Get:Actor Count Volume Mesh Location Struct Structure)
   > Add(
    TargetArray:Array of Vectors=Break Actor Count Volume Mesh Location Struct!(ActorCountVolumeMeshLocationStruct:"Actor Count Volume Mesh Location Struct Structure (by ref)"=@n13) -> (Locations_3_#1:Array of Vectors) {is=title K2Node_BreakStruct}
    NewItem:"Vector (by ref)"=Get l_actorCenterLocation() -> (l_actorCenterLocation:Vector)
   ) {is=title K2Node_CallArrayFunction}
   > Add(
    TargetMap:Map of Names to Actor Count Volume Mesh Location Struct Structures=Get l_classLocationSet()
    Key:"Name (by ref)"=Get l_classRootPath() -> (l_classRootPath=None)
    Value:"Actor Count Volume Mesh Location Struct Structure (by ref)"=Get l_locationEditStruct() -> (l_locationEditStruct:Actor Count Volume Mesh Location Struct Structure)
   ) {is=title K2Node_CallFunction}
   n14 = Get l_actorCenterLocation()
   Add(
    execute=@n12.else
    TargetMap=Get l_classLocationSet()
    Key=Get l_classRootPath() -> (l_classRootPath=None)
    Value=Make Actor Count Volume Mesh Location Struct!(Locations_3_#1:Array of Vectors=Make Array("[0]":Vector=@n14) -> (Array:Array of Vectors)) -> (ActorCountVolumeMeshLocationStruct:Actor Count Volume Mesh Location Struct Structure) {is=title K2Node_MakeStruct}
   )
   n15 = Get l_classSet()
   n16 = Values!(execute=@n1.then_1, TargetMap:Map of Soft Object Path Structures to Integers=@n15) -> (then, Values:Array of Integers)
   > n17 = Keys!(execute, TargetMap:Map of Soft Object Path Structures to Integers=@n15) -> (then, Keys:Array of Soft Object Path Structures)
   > n18 = For Each Loop(Array:Array of Soft Object Path Structures=@n17) -> (Array Element:Soft Object Path Structure)
   > Set l_class(l_class=@n18)
   > Set l_count(l_count:int="Get (a copy)"(Array:Array of Integers=@n16, Dimension 1=@n18.Array Index) -> (Output:int)) -> (Output_Get:int=0)
   n19 = Get l_class()
   n20 = String To Name(InString=Split(SourceString=BreakSoftObjectPath(InSoftObjectPath=@n19), InStr=., SearchDir=FromStart))
   n21 = MakeTopLevelAssetPath!(PackageName:String=/Script/Engine.Blueprint, AssetName:String) -> (ReturnValue:Top Level Asset Path Structure)
   > n22 = GetAssets!(
    execute
    self:Asset Registry Interface=GetAssetRegistry()
    Filter:"ARFilter Structure (by ref)"=MakeARFilter!(
     PackageNames:Array of Names=Make Array("[0]":Name=@n20) -> (Array:Array of Names)
     PackagePaths:Array of Names
     SoftObjectPaths:Array of Soft Object Path Structures
     ClassPaths:Array of Top Level Asset Path Structures=Make Array("[0]":Top Level Asset Path Structure=@n21) -> (Array:Array of Top Level Asset Path Structures)
     RecursiveClassPathsExclusionSet:Set of Top Level Asset Path Structures
     ClassNames:Array of Names
     RecursiveClassesExclusionSet:Set of Names
     bRecursivePaths:bool=false
     bRecursiveClasses:bool=true
     bIncludeOnlyOnDiskAssets:bool=false
    ) -> (ReturnValue:ARFilter Structure)
    bSkipARFilteredAssets:bool=true
   ) -> (then, OutAssetData:Array of Asset Data Structures, ReturnValue:bool)
   > n23 = For Each Loop(Array:Array of Asset Data Structures=@n22) -> (Array Element:Asset Data Structure)
   n24 = Break Asset Data(AssetData=@n23)
   > Set l_bpName(l_bpName:Name=@n24) -> (Output_Get:Name=None)
   > n25 = Get Dependencies!(
    execute
    self:Asset Registry Interface=GetAssetRegistry()
    PackageName:Name=@n24
    DependencyOptions:"Asset Registry Dependency Options Structure (by ref)"=Make Asset Registry Dependency Options(bIncludeSoftPackageReferences=false)
   ) -> (then, OutDependencies:Array of Names, ReturnValue:bool) {is=title K2Node_CallFunction}
   > Branch(Condition=Is Valid Index!(TargetArray:Array of Names=@n25, IndexToTest:int) -> (ReturnValue:bool))
   n26 = MakeTopLevelAssetPath(PackageName=/Script/Engine, AssetName=StaticMesh)
   > n27 = GetAssets(
    self=GetAssetRegistry()
    Filter=MakeARFilter(PackageNames=@n25, ClassPaths=Make Array("[0]"=@n26), bRecursivePaths=false, bRecursiveClasses=false, bIncludeOnlyOnDiskAssets=false)
    bSkipARFilteredAssets=true
   )
   > n28 = For Each Loop(Array=@n27)
   > Set l_name(l_name:Name=Break Asset Data(AssetData=@n28)) -> (Output_Get:Name=None)
   > n29 = NoLoad_GetStaticMeshCollisionData!(execute, AssetData:Asset Data Structure=@n28, __WorldContext:Object Reference=Get l_world()) -> (then, SectionsWithCollision:int=0, CollisionPrimCount:int=0, CollisionComplexity:String)
   > Branch(Condition="integer > integer"(A=@n29))
   n30 = Find(
    TargetMap:Map of Names to Actor Count Volume Mesh Table Struct Structures=Get l_meshCountMap() -> (l_meshCountMap:Map of Names to Actor Count Volume Mesh Table Struct Structures)
    Key=Get l_name() -> (l_name:Name=None)
   ) -> (Value:Actor Count Volume Mesh Table Struct Structure)
   > n31 = Branch(Condition=@n30.ReturnValue)
   > n32 = Set l_editStruct(l_editStruct:Actor Count Volume Mesh Table Struct Structure=@n30) -> (Output_Get:Actor Count Volume Mesh Table Struct Structure)
   n33 = Break Actor Count Volume Mesh Table Struct!(ActorCountVolumeMeshTableStruct:"Actor Count Volume Mesh Table Struct Structure (by ref)"=@n32) -> (
    AssetName_15_#2:Name
    Count_16_#3:int
    AssetRefs_23_#4:Array of Strings
    MaxCount_27_#5:int
    Locations_31_#6:Array of Vectors
   ) {is=title K2Node_BreakStruct}
   > n34 = Set l_editArray(l_editArray:Array of Strings=@n33.AssetRefs_23_#4) -> (Output_Get:Array of Strings)
   > Add(TargetArray:Array of Strings=@n34, NewItem:"String (by ref)"="To String (Name)"(InName=Get l_bpName() -> (l_bpName:Name=None))) {is=title K2Node_CallArrayFunction}
   > n35 = Set l_newCount(l_newCount:int=int + int(A=@n33.Count_16_#3, B=Get l_count() -> (l_count:int=0))) -> (Output_Get:int=0)
   > n36 = Set members in Actor Count Volume Mesh Table Struct!(
    execute:Exec
    StructRef:"Actor Count Volume Mesh Table Struct Structure (by ref)"=Get l_editStruct() -> (l_editStruct:Actor Count Volume Mesh Table Struct Structure)
    Count_16_#3:int=@n35
    AssetRefs_23_#4:Array of Strings=Get l_editArray() -> (l_editArray:Array of Strings)
   ) -> (then:Exec, StructOut:"Actor Count Volume Mesh Table Struct Structure (by ref)") {is=title K2Node_SetFieldsInStruct}
   > Add(
    TargetMap:Map of Names to Actor Count Volume Mesh Table Struct Structures=Get l_meshCountMap()
    Key=Get l_name() -> (l_name=None)
    Value:"Actor Count Volume Mesh Table Struct Structure (by ref)"=@n36
   ) {is=title K2Node_CallFunction}
   > n37 = Set l_largest(l_largest:int="Max (Integer)"(A=Get l_newCount() -> (l_newCount:int=0), B=Get l_largest() -> (l_largest:int=0))) -> (Output_Get:int=0)
   n38 = Find(TargetMap=Get l_meshCountMap(), Key=Get l_name() -> (l_name=None))
   Branch(execute=[@n46.then, @n37.then], Condition=@n38.ReturnValue)
   n39 = BreakSoftObjectPath(InSoftObjectPath=Get l_class())
   n40 = Find(
    TargetMap:Map of Names to Actor Count Volume Mesh Location Struct Structures=Get l_classLocationSet()
    Key=String To Name(InString=Split(SourceString=@n39, InStr=., SearchDir=FromStart))
   ) -> (Value:Actor Count Volume Mesh Location Struct Structure)
   > Branch(Condition=@n40.ReturnValue)
   n41 = Break Actor Count Volume Mesh Location Struct(ActorCountVolumeMeshLocationStruct=@n40)
   > n42 = Set members in Actor Count Volume Mesh Table Struct!(execute, StructRef=@n38, Locations_31_#6:Array of Vectors=@n41) -> (then, StructOut)
   > Add(TargetMap=Get l_meshCountMap(), Key=Get l_name() -> (l_name=None), Value=@n42)
   > n43 = For Each Loop(Array:Array of Vectors=@n41) -> (Array Element:Vector)
   > Log String!(
    Execute:Exec
    Blueprint Log Level:BlueprintLogLevel Enum=NewEnumerator3
    Log Level Instance:BlueprintLogLevel Enum=NewEnumerator0
    Log String:String="To String (Vector)"!(InVec:Vector=@n43) -> (ReturnValue:String)
    Print to Screen:bool=true
   ) -> (Then:Exec) {is=title K2Node_MacroInstance}
   n44 = "To String (Name)"(InName=Get l_bpName() -> (l_bpName=None))
   n45 = Get l_name() -> (l_name=None)
   n46 = Add(
    execute=@n31.else
    TargetMap=Get l_meshCountMap()
    Key=@n45
    Value=Make Actor Count Volume Mesh Table Struct!(
     AssetName_15_#2:Name=@n45
     Count_16_#3:int=Get l_count() -> (l_count=0)
     AssetRefs_23_#4:Array of Strings=Make Array("[0]":String=@n44) -> (Array:Array of Strings)
     MaxCount_27_#5:int=-1
     Locations_31_#6:Array of Vectors
    ) -> (ActorCountVolumeMeshTableStruct:Actor Count Volume Mesh Table Struct Structure) {is=title K2Node_MakeStruct}
   )
   n47 = Get l_meshCountMap()
   n48 = Values(execute=@n1.then_2, TargetMap:Map of Names to Actor Count Volume Mesh Table Struct Structures=@n47) -> (Values:Array of Actor Count Volume Mesh Table Struct Structures)
   > n49 = Keys(TargetMap:Map of Names to Actor Count Volume Mesh Table Struct Structures=@n47) -> (Keys:Array of Names)
   > n50 = For Each Loop(Array:Array of Actor Count Volume Mesh Table Struct Structures=@n48) -> (Array Element:Actor Count Volume Mesh Table Struct Structure)
   > n51 = Set members in Actor Count Volume Mesh Table Struct!(execute, StructRef=@n50, MaxCount_27_#5:int=Get l_largest() -> (l_largest=0)) -> (then, StructOut)
   > Add(
    TargetMap=Get l_meshCountMap()
    Key="Get (a ref)"(Array:Array of Names=@n49, Dimension 1=@n50.Array Index) -> (Output:"Name (by ref)")
    Value=@n51
   )
   n52 = Sequence(execute=@n1.then_3)
   > n53 = ConvertFilePathToAssetPath!(execute, InPath:String=GameUserDeveloperDir!() -> (ReturnValue:String)) -> (then, OutPath:String, ReturnValue:bool)
   > n54 = Set l_tablePackageName(l_tablePackageName:String=Append!(A=@n53, B, C:String=Get l_tableName() -> (l_tableName:String)) -> (ReturnValue)) -> (Output_Get:String)
   > n55 = DoesAssetExist(AssetPath=@n54)
   > n56 = Branch(Condition=@n55)
   > n57 = GetAssetsByPackageName!(
    execute
    self:Asset Registry Interface=GetAssetRegistry()
    PackageName:Name=String To Name(InString=Get l_tablePackageName() -> (l_tablePackageName:String))
    bIncludeOnlyOnDiskAssets:bool=false
    bSkipARFilteredAssets:bool=true
   ) -> (then, OutAssetData:Array of Asset Data Structures, ReturnValue:bool)
   > Branch(Condition=Is Valid Index(TargetArray:Array of Asset Data Structures=@n57))
   > n58 = Cast To DataTable!(execute, Object=GetAsset(InAssetData="Get (a copy)"(Array:Array of Asset Data Structures=@n57) -> (Output:Asset Data Structure))) -> (then, CastFailed, AsData Table:Data Table Object Reference) {is=title K2Node_DynamicCast}
   > Set l_table(l_table:Data Table Object Reference=@n58) -> (Output_Get:Data Table Object Reference)
   n59 = Execute Python Script!(
    execute=@n56.else
    PythonScript="import unreal\r\n\r\nfactory = unreal.DataTableFactory()\r\nfactory.struct = in_struct_asset"
    in_struct_asset:Object Reference=Get l_struct() -> (l_struct:Object Reference)
   ) -> (then, ReturnValue, factory:Factory Object Reference) {inputs [in_struct_asset], outputs [factory]}
   > Branch(Condition=@n59)
   > n60 = GetAssetTools()
   > n61 = CreateAsset(self=@n60, AssetName=Get l_tableName(), PackagePath=@n53, AssetClass=/Script/Engine.DataTable, Factory=@n59.factory)
   > n62 = Cast To DataTable(Object=@n61)
   > Set l_table(l_table=@n62)
   Is Valid(exec=@n52.then_1, InputObject=Get l_table() -> (l_table:Data Table Object Reference))
   > Set l_locationArrayString(l_locationArrayString:String) -> (Output_Get:String)
   > Set l_arrayString(l_arrayString:String) -> (Output_Get:String)
   n63 = Get l_meshCountMap()
   > n64 = Values(TargetMap=@n63)
   > n65 = Keys(TargetMap=@n63)
   > n66 = For Each Loop(Array:Array of Names=@n65) -> (Array Element:Name)
   Fill Data Table from CSV String!(
    execute=@n66.Completed
    DataTable:Data Table Object Reference=Get l_table()
    CSVString:String=Get l_csv() -> (l_csv:String)
    ImportRowStruct:Script Struct Object Reference
   ) -> (then, ReturnValue:bool)
   > SyncBrowserToObjects(AssetPaths=Make Array("[0]"=Get l_tablePackageName()))
   Set l_locationArrayString(execute=@n66.LoopBody)
   > Set l_arrayString()
   > n67 = Sequence()
   n68 = Break Actor Count Volume Mesh Table Struct(ActorCountVolumeMeshTableStruct="Get (a ref)"(Array:Array of Actor Count Volume Mesh Table Struct Structures=@n64, Dimension 1=@n66.Array Index) -> (Output:"Actor Count Volume Mesh Table Struct Structure (by ref)"))
   > n69 = For Each Loop(Array:Array of Vectors=@n68.Locations_31_#6) -> (Array Element:Vector)
   n70 = "Equal (Integer)"(A=Last Index!(TargetArray:Array of Vectors=@n68.Locations_31_#6) -> (ReturnValue=0), B=@n69.Array Index)
   n71 = BreakVector(InVec=@n69)
   > Set l_locationArrayString(l_locationArrayString=Append!(
     A=Get l_locationArrayString() -> (l_locationArrayString:String)
     B="(X="
     C="To String (Float)"(InDouble=@n71)
     D:String=","
     E:String="Y="
     F:String="To String (Float)"(InDouble=@n71.Y)
     G:String=","
     H:String="Z="
     I:String="To String (Float)"(InDouble=@n71.Z)
     J:String=")"
     K:String=SelectString(B=",", bPickA=@n70)
    ) -> (ReturnValue))
   n72 = For Each Loop(Exec=@n67.then_1, Array:Array of Strings=@n68.AssetRefs_23_#4) -> (Array Element:String)
   n73 = "To String (Integer)"(InInt=Get l_largest() -> (l_largest=0))
   n74 = "To String (Name)"(InName=@n66)
   n75 = "To String (Integer)"(InInt=@n68.Count_16_#3)
   n76 = Get l_locationArrayString()
   n77 = Get l_arrayString() -> (l_arrayString:String)
   Set l_csv(
    execute=@n72.Completed
    l_csv=Append!(
     A=Get l_csv()
     B=Append!(
      A=@n74
      B=","
      C=@n74
      D:String=","
      E:String=@n75
      F:String=",\"("
      G:String=@n77
      H:String=")\""
      I:String=","
      J:String=@n73
      K:String=",\"("
      L:String=@n76
      M:String=")\""
     ) -> (ReturnValue)
     C="\r\n"
    ) -> (ReturnValue)
   )
   n78 = "Equal (Integer)"(A=@n72.Array Index, B=Last Index(TargetArray:Array of Strings=@n68.AssetRefs_23_#4) -> (ReturnValue=0))
   Set l_arrayString(
    execute=@n72.LoopBody
    l_arrayString=Append!(A=Get l_arrayString(), B="\"\"", C=@n72, D="\"\"", E=SelectString(B=",", bPickA=@n78)) -> (ReturnValue)
   )
   n79 = Values(execute=@n1.then_5, TargetMap=Get l_meshCountMap())
   > PopulateUI!(execute, self, Data:Array of Actor Count Volume Mesh Table Struct Structures=@n79) -> (then)
   > n80 = Values(TargetMap=Get l_meshCountMap())
   > n81 = For Each Loop(Array:Array of Actor Count Volume Mesh Table Struct Structures=@n80) -> (Array Element:Actor Count Volume Mesh Table Struct Structure)
   > n82 = For Each Loop(Array:Array of Vectors=Break Actor Count Volume Mesh Table Struct(ActorCountVolumeMeshTableStruct=@n81).Locations_31_#6) -> (Array Element:Vector)
   > Log String(Blueprint Log Level=NewEnumerator3, Log Level Instance=NewEnumerator0, Log String="To String (Vector)"(InVec=@n82), Print to Screen=true)
  ]}
 OnCountUpdated {nodes [OnCountUpdated!() -> (then, actor:EUA Actor Count Volume Object Reference) {is=title K2Node_FunctionEntry}]}
 PopulateUI {nodes [
   n1 = PopulateUI!() -> (then, Data:Array of Actor Count Volume Mesh Table Struct Structures) {is=title K2Node_FunctionEntry}
   > n2 = SpawnAndRegisterTab(
    self=Get EditorUtilitySubsystem()
    InBlueprint=/Game/NeverCook/EditorUtilities/PerfTools/ActorCountVolume/Widgets/EUW_ActorCountVolumeMeshUsage
   )
   > n3 = Cast To EUW_ActorCountVolumeMeshUsage!(execute, Object=@n2) -> (then, CastFailed, AsEUW Actor Count Volume Mesh Usage:EUW Actor Count Volume Mesh Usage Object Reference) {is=title K2Node_DynamicCast}
   > setMeshData!(execute, self:EUW Actor Count Volume Mesh Usage Object Reference=@n3, MeshData:Array of Actor Count Volume Mesh Table Struct Structures=@n1) -> (then)
  ]}
 RefreshActorCountWithFiltering {nodes [
   RefreshActorCountWithFiltering() {is=title K2Node_FunctionEntry}
   > GetActorCount!(execute, self, AllowClassFiltering:bool=true) -> (then) {is=title K2Node_CallFunction}
  ]}
 SetActorCount {nodes [
   n1 = SetActorCount!() -> (then, CurrentActorCount:int) {is=title K2Node_FunctionEntry}
   > n2 = Set CurrentActorCount(CurrentActorCount:int=@n1) -> (Output_Get:int=0)
   > SetDefaultCustomPrimitiveDataFloat!(
    execute
    self:Primitive Component Object Reference=Get Cube() -> (Cube:Static Mesh Component Object Reference)
    DataIndex:int=0
    Value:float="To Float (Integer)"(InInt=@n2)
   ) -> (then)
   n3 = "To String (Integer)"(InInt=@n2)
   > SetDefaultCustomPrimitiveDataFloat(self=Get Cube(), DataIndex=1, Value="To Float (Integer)"(InInt=Len(S=@n3)))
  ]}
 ShowWidget {nodes [
   ShowWidget() {is=title K2Node_FunctionEntry}
   > SpawnAndRegisterTab(self=Get EditorUtilitySubsystem(), InBlueprint=&/EUW_ActorCountVolumeStats)
  ]}
 TestDataLayers {nodes [
   n1 = TestDataLayers!() -> (then, DataLayers:Array of Soft Object Path Structures) {is=title K2Node_FunctionEntry}
   > n2 = For Each Loop with Break!(Exec:Exec, Array:Array of Soft Object Path Structures=@n1, Break:Exec) -> (LoopBody:Exec, Array Element:Soft Object Path Structure, Array Index:int, Completed:Exec)
   Return Node!(execute=@n2.Completed, valid:bool=true)
   > n3 = LoadAsset_Blocking(Asset=To Soft Object Reference(SoftObjectPath=@n2))
   > n4 = Cast To DataLayerAsset!(execute, Object=@n3) -> (then, CastFailed, AsData Layer Asset:Data Layer Asset Object Reference) {is=title K2Node_DynamicCast}
   n5 = GetDataLayerInstance!(
    self:Data Layer Editor Subsystem Object Reference=Get DataLayerEditorSubsystem() -> (ReturnValue:Data Layer Editor Subsystem Object Reference) {is=title K2Node_GetEditorSubsystem}
    DataLayerAsset:Data Layer Asset Object Reference=@n4
   ) -> (ReturnValue:Data Layer Instance Object Reference)
   > Is Valid(InputObject=@n5)
   > n6 = Branch(Condition=IsVisible(self:Data Layer Instance Object Reference=@n5))
   Return Node(execute=@n6.else, valid=false)
  ]}
 UserConstructionScript {nodes [
   Construction Script()
   > n1 = Branch(Condition=IsUnattended!() -> (ReturnValue:bool))
   GetActorCount(execute=@n1.else, AllowClassFiltering=false)
  ]}
]
variables [
 CachedSDPCollectionNames {type="array<string>", default []}
 CurrentActorCount {type=integer}
 CurrentActorDescs {type="array<ActorDesc>", default []}
 IncludeEditorOnly {type=boolean}
 IncludeSDP {type=boolean}
 OnCountUpdated {default=()}
 TransientClassFilter {type="array<Class>", default []}
]
